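# Additive deformation Control Rig for the ZirconSweep medium male body, correcting jacket, hood, sack bag and strap bones as the character moves.
# Weighted parent constraints anchor accessory and jacket bones to body bones.
# Spherical pose readers on thighs, head, left arm and clavicle weight transform offsets, with a mirror helper for right-side jacket offsets.
# Crouch and baseball or knee slide flags select dedicated jacket and bag adjustments.
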ControlRigBlueprint /BRCosmetics/Characters/Player/Male/Medium/Bodies/M_MED_ZirconSweep/Meshes/M_MED_ZirconSweep_AdditiveDeformation_CtrlRig
graphs [
 ContainedGraph {nodes [
   n1 = RigVMFunctionEntryNode()
   > SequenceAggregate = RigVMAggregateNode()
   &RigUnit_ModifyTransforms(
    ItemToModify="((Item=(Type=Bone,Name=\"jacket_l_front_01_root\"),Transform=(Rotation=(X=0,Y=0,Z=0,W=1),Translation=(X=0,Y=0,Z=0),Scale3D=(X=1,Y=1,Z=1))),(Item=(Type=Bone,Name=\"jacket_l_side_01_root\"),Transform=(Rotation=(X=0,Y=0,Z=0,W=1),Translation=(X=0,Y=0,Z=0),Scale3D= ...(truncated)"
    Weight=@n1.thigh_l_alpha
    ExecutePin=@SequenceAggregate.A
    ItemToModify.1.Transform=RigVMVariableNode() {variable=jacket_side_leg_fwd}.Value
    ItemToModify.0.Transform=RigVMVariableNode() {variable=jacket_front_leg_fwd}.Value
   )
   > Deform_GetMirrorTransform_FNC_2 = RigVMFunctionReferenceNode(transform=RigVMVariableNode() {variable=jacket_front_leg_fwd}.Value)
   > Deform_GetMirrorTransform_FNC_2_1 = RigVMFunctionReferenceNode(transform=RigVMVariableNode() {variable=jacket_side_leg_fwd}.Value)
   > &RigUnit_ModifyTransforms(
    ItemToModify="((Item=(Type=Bone,Name=\"jacket_r_front_01_root\"),Transform=(Rotation=(X=0,Y=0,Z=0,W=1),Translation=(X=0,Y=0,Z=0),Scale3D=(X=1,Y=1,Z=1))),(Item=(Type=Bone,Name=\"jacket_r_side_01_root\"),Transform=(Rotation=(X=0,Y=0,Z=0,W=1),Translation=(X=0,Y=0,Z=0),Scale3D= ...(truncated)"
    ItemToModify.0.Transform=@Deform_GetMirrorTransform_FNC_2.mirrorTransform
    ItemToModify.1.Transform=@Deform_GetMirrorTransform_FNC_2_1.mirrorTransform
    Weight=@n1.thigh_r_alpha
   )
   n2 = &RigVMFunction_ControlFlowBranch(
    Condition=&RigVMFunction_MathBoolOr(A=RigVMVariableNode() {variable=baseball_sliding}.Value, B=RigVMVariableNode() {variable=knee_sliding}.Value).Result
    ExecuteContext=@SequenceAggregate.B
   )
   Cast = &RigVMFunction_MathBoolToFloat(Value=RigVMVariableNode() {variable=baseball_sliding}.Value)
   &RigUnit_ModifyTransforms(
    ItemToModify="((Item=(Type=Bone,Name=\"jacket_c_rear_01_root\"),Transform=(Rotation=(X=0,Y=0,Z=0,W=1),Translation=(X=0,Y=0,Z=0),Scale3D=(X=1,Y=1,Z=1))),(Item=(Type=Bone,Name=\"jacket_l_front_01_root\"),Transform=(Rotation=(X=0,Y=0,Z=0,W=1),Translation=(X=0,Y=0,Z=0),Scale3D= ...(truncated)"
    ExecutePin=@n2.True
    Weight=@Cast.Result
    ItemToModify.3.Transform=RigVMVariableNode() {variable=baseball_slide_jacket_r_front}.Value
    ItemToModify.4.Transform=RigVMVariableNode() {variable=baseball_slide_jacket_r_rear}.Value
    ItemToModify.0.Transform=RigVMVariableNode() {variable=baseball_slide_jacket_c_rear}.Value
    ItemToModify.1.Transform=RigVMVariableNode() {variable=baseball_slide_jacket_l_front}.Value
    ItemToModify.2.Transform=RigVMVariableNode() {variable=baseball_slide_jacket_l_rear}.Value
   )
   > &RigUnit_ModifyTransforms(
    ItemToModify="((Item=(Type=Bone,Name=\"jacket_c_rear_01_root\"),Transform=(Rotation=(X=0,Y=0,Z=0,W=1),Translation=(X=0,Y=0,Z=0),Scale3D=(X=1,Y=1,Z=1))),(Item=(Type=Bone,Name=\"jacket_l_front_01_root\"),Transform=(Rotation=(X=0,Y=0,Z=0,W=1),Translation=(X=0,Y=0,Z=0),Scale3D= ...(truncated)"
    Weight=&RigVMFunction_MathBoolToFloat(Value=RigVMVariableNode() {variable=knee_sliding}.Value).Result
    ItemToModify.0.Transform=RigVMVariableNode() {variable=knee_slide_jacket_c_rear}.Value
    ItemToModify.1.Transform=RigVMVariableNode() {variable=knee_slide_jacket_l_front}.Value
    ItemToModify.2.Transform=RigVMVariableNode() {variable=knee_slide_jacket_l_rear}.Value
    ItemToModify.3.Transform=RigVMVariableNode() {variable=knee_slide_jacket_r_front}.Value
    ItemToModify.4.Transform=RigVMVariableNode() {variable=knee_slide_jacket_r_rear}.Value
   )
   RigVMFunctionReturnNode(ExecuteContext=@SequenceAggregate.C)
  ]}
 ContainedGraph {nodes [
   n1 = RigVMFunctionEntryNode()
   SequenceAggregate = &RigVMFunction_Sequence(ExecuteContext=@n1.ExecuteContext)
   SequenceAggregate_1 = &RigVMFunction_Sequence(ExecuteContext=@SequenceAggregate.B)
   RigVMFunctionReturnNode(A=@SequenceAggregate.A, B=@SequenceAggregate_1.A, C=@SequenceAggregate_1.B)
  ]}
 ContainedGraph {nodes [
   n1 = RigVMFunctionEntryNode()
   > &RigUnit_ModifyTransforms(
    ItemToModify="((Item=(Type=Bone,Name=\"dyn_sackbag_root\"),Transform=(Rotation=(X=0,Y=0,Z=0,W=1),Translation=(X=0,Y=0,Z=0),Scale3D=(X=1,Y=1,Z=1))),(Item=(Type=Bone,Name=\"dyn_sackbelt_r_hip_root\"),Transform=(Rotation=(X=0,Y=0,Z=0,W=1),Translation=(X=0,Y=0,Z=0),Scale3D=(X=1 ...(truncated)"
    Weight=@n1.thigh_r_alpha
    ItemToModify.0.Transform=RigVMVariableNode() {variable=sackbag_thigh_r_fwd}.Value
    ItemToModify.1.Transform=RigVMVariableNode() {variable=sackbelt_hip_thigh_r_fwd}.Value
    ItemToModify.2.Transform=RigVMVariableNode() {variable=sackbelt_lowerback_thigh_r_fwd}.Value
   )
   > RigVMFunctionReturnNode()
  ]}
 ContainedGraph {nodes [
   n1 = RigVMFunctionEntryNode()
   n2 = /Script/RigVM.RigVMFunction_MathTransformToEulerTransform(Value=@n1.transform) {is=struct RigVMUnitNode, template="Cast::Execute(in Value,out Result)"}
   > RigVMFunctionReturnNode(
    mirrorTransform=(Rotation=(X=0,Y=0,Z=0,W=1),Translation=(X=0,Y=0,Z=0),Scale3D=(X=0,Y=0,Z=0))
    mirrorTransform.Rotation=/Script/RigVM.RigVMFunction_MathQuaternionFromRotatorV2(Value=@n2.Result.Rotation) {is=struct RigVMUnitNode, template="Cast::Execute(in Value,out Result)"}.Result
    mirrorTransform.Translation=/Script/RigVM.RigVMFunction_MathVectorScale(Factor=-1, Value=@n2.Result.Location) {is=struct RigVMUnitNode, template="Scale::Execute(in Value,in Factor,out Result)"}.Result
    mirrorTransform.Scale3D=@n2.Result.Scale
   )
  ]}
 ContainedGraph {nodes [
   RigVMFunctionEntryNode()
   > &RigUnit_ParentConstraint(
    AdvancedSettings=(RotationOrderForFilter=XZY)
    Child=(Type=Bone,Name="dyn_sackbelt_l_pec_root")
    Parents="((Item=(Type=Bone,Name=\"deform_pec_l\"),Weight=0.0102),(Item=(Type=Bone,Name=\"clavicle_l\"),Weight=0.0415),(Item=(Type=Bone,Name=\"pelvis\"),Weight=0.0711),(Item=(Type=Bone,Name=\"spine_04\"),Weight=0.0883),(Item=(Type=Bone,Name=\"spine_01\"),Weight=0.1436),(Item= ...(truncated)"
    Weight=1
   )
   > &RigUnit_ParentConstraint(
    AdvancedSettings=(RotationOrderForFilter=XZY)
    Child=(Type=Bone,Name="dyn_sackbelt_l_scap_root")
    Parents="((Item=(Type=Bone,Name=\"upperarm_twist_02_l\"),Weight=0.01),(Item=(Type=Bone,Name=\"spine_04\"),Weight=0.0569),(Item=(Type=Bone,Name=\"deform_upperarm_up_l\"),Weight=0.0609),(Item=(Type=Bone,Name=\"spine_05\"),Weight=0.1035),(Item=(Type=Bone,Name=\"upperarm_twist_ ...(truncated)"
    Weight=1
   )
   > &RigUnit_ParentConstraint(
    AdvancedSettings=(RotationOrderForFilter=XZY)
    Child=(Type=Bone,Name="dyn_sackbelt_l_pec_root")
    Parents="((Item=(Type=Bone,Name=\"spine_02\"),Weight=0.0263),(Item=(Type=Bone,Name=\"upperarm_twist_01_l\"),Weight=0.034),(Item=(Type=Bone,Name=\"deform_upperarm_fr_l\"),Weight=0.043),(Item=(Type=Bone,Name=\"spine_03\"),Weight=0.076),(Item=(Type=Bone,Name=\"spine_05\"),Weigh ...(truncated)"
    Weight=1
   )
   > &RigUnit_ParentConstraint(
    AdvancedSettings=(RotationOrderForFilter=XZY)
    Child=(Type=Bone,Name="dyn_sackbag_root")
    Parents=((Item=(Type=Bone,Name="pelvis"),Weight=0.0719),(Item=(Type=Bone,Name="spine_04"),Weight=0.0992),(Item=(Type=Bone,Name="spine_01"),Weight=0.1281),(Item=(Type=Bone,Name="spine_02"),Weight=0.3078),(Item=(Type=Bone,Name="spine_03"),Weight=0.3929))
    Weight=0
   )
   > &RigUnit_ParentConstraint(
    AdvancedSettings=(RotationOrderForFilter=XZY)
    Child=(Type=Bone,Name="dyn_sackbelt_r_hip_root")
    Parents="((Item=(Type=Bone,Name=\"spine_03\"),Weight=0.041),(Item=(Type=Bone,Name=\"jacket_r_rear_01\"),Weight=0.0553),(Item=(Type=Bone,Name=\"jacket_r_front_02\"),Weight=0.0568),(Item=(Type=Bone,Name=\"spine_01\"),Weight=0.0692),(Item=(Type=Bone,Name=\"spine_02\"),Weight=0. ...(truncated)"
    Weight=1
   )
   > &RigUnit_ParentConstraint(
    AdvancedSettings=(RotationOrderForFilter=XZY)
    Child=(Type=Bone,Name="dyn_sackbelt_r_lowerback_root")
    Parents="((Item=(Type=Bone,Name=\"spine_05\"),Weight=0.0057),(Item=(Type=Bone,Name=\"jacket_r_rear_02\"),Weight=0.0265),(Item=(Type=Bone,Name=\"spine_04\"),Weight=0.0924),(Item=(Type=Bone,Name=\"spine_01\"),Weight=0.0959),(Item=(Type=Bone,Name=\"jacket_r_rear_01\"),Weight=0. ...(truncated)"
    Weight=1
    bMaintainOffset=true
   )
   > &RigUnit_ParentConstraint(
    AdvancedSettings=(RotationOrderForFilter=XZY)
    Child=(Type=Bone,Name="dyn_sackbelt_l_lowerback_root")
    Parents="((Item=(Type=Bone,Name=\"clavicle_l\"),Weight=0.0029),(Item=(Type=Bone,Name=\"pelvis\"),Weight=0.0896),(Item=(Type=Bone,Name=\"spine_04\"),Weight=0.0973),(Item=(Type=Bone,Name=\"spine_01\"),Weight=0.1303),(Item=(Type=Bone,Name=\"spine_02\"),Weight=0.2974),(Item=(Typ ...(truncated)"
    Weight=1
   )
   > &RigUnit_ParentConstraint(
    AdvancedSettings=(RotationOrderForFilter=XZY)
    Child=(Type=Bone,Name="dyn_l_leg_strap_root")
    Parents="((Item=(Type=Bone,Name=\"deform_knee_out_l\"),Weight=0.0031),(Item=(Type=Bone,Name=\"deform_knee_in_l\"),Weight=0.0054),(Item=(Type=Bone,Name=\"thigh_l\"),Weight=0.0389),(Item=(Type=Bone,Name=\"calf_twist_02_l\"),Weight=0.2397),(Item=(Type=Bone,Name=\"calf_l\"),Weig ...(truncated)"
    Weight=1
   )
   > &RigUnit_ParentConstraint(
    AdvancedSettings=(RotationOrderForFilter=XZY)
    Child=(Type=Bone,Name="dyn_thigh_l_zipper_root")
    Parents="((Item=(Type=Bone,Name=\"deform_knee_in_l\"),Weight=0.0114),(Item=(Type=Bone,Name=\"deform_knee_out_l\"),Weight=0.0691),(Item=(Type=Bone,Name=\"calf_l\"),Weight=0.0817),(Item=(Type=Bone,Name=\"thigh_twist_01_l\"),Weight=0.2044),(Item=(Type=Bone,Name=\"thigh_l\"),Wei ...(truncated)"
    Weight=1
   )
   > &RigUnit_ParentConstraint(
    AdvancedSettings=(RotationOrderForFilter=XZY)
    Child=(Type=Bone,Name="dyn_thigh_r_zipper_root")
    Parents="((Item=(Type=Bone,Name=\"deform_knee_in_r\"),Weight=0.0114),(Item=(Type=Bone,Name=\"deform_knee_out_r\"),Weight=0.0691),(Item=(Type=Bone,Name=\"calf_r\"),Weight=0.0817),(Item=(Type=Bone,Name=\"thigh_twist_01_r\"),Weight=0.2044),(Item=(Type=Bone,Name=\"thigh_r\"),Wei ...(truncated)"
    Weight=1
   )
   > &RigUnit_ParentConstraint(
    AdvancedSettings=(RotationOrderForFilter=XZY)
    Child=(Type=Bone,Name="dyn_l_arm_strap_root")
    Parents=((Item=(Type=Bone,Name="upperarm_twist_02_l"),Weight=0.3),(Item=(Type=Bone,Name="upperarm_twist_01_l"),Weight=0.7))
    Weight=1
   )
   > &RigUnit_ParentConstraint(
    AdvancedSettings=(RotationOrderForFilter=XZY)
    Child=(Type=Bone,Name="dyn_r_arm_strap_root")
    Parents=((Item=(Type=Bone,Name="upperarm_twist_02_r"),Weight=0.3),(Item=(Type=Bone,Name="upperarm_twist_01_r"),Weight=0.7))
    Weight=1
   )
   > &RigUnit_ParentConstraint(
    AdvancedSettings=(RotationOrderForFilter=XZY)
    Child=(Type=Bone,Name="jacket_l_side_01")
    Parents="((Item=(Type=Bone,Name=\"clavicle_l\"),Weight=0.0207),(Item=(Type=Bone,Name=\"spine_04\"),Weight=0.0462),(Item=(Type=Bone,Name=\"pelvis\"),Weight=0.1573),(Item=(Type=Bone,Name=\"spine_01\"),Weight=0.1902),(Item=(Type=Bone,Name=\"spine_03\"),Weight=0.2529),(Item=(Typ ...(truncated)"
    Weight=1
   )
   > &RigUnit_ParentConstraint(
    AdvancedSettings=(RotationOrderForFilter=XZY)
    Child=(Type=Bone,Name="jacket_l_rear_01")
    Parents="((Item=(Type=Bone,Name=\"clavicle_l\"),Weight=0.0028),(Item=(Type=Bone,Name=\"spine_04\"),Weight=0.0501),(Item=(Type=Bone,Name=\"pelvis\"),Weight=0.1803),(Item=(Type=Bone,Name=\"spine_01\"),Weight=0.1871),(Item=(Type=Bone,Name=\"spine_03\"),Weight=0.2627),(Item=(Typ ...(truncated)"
    Weight=1
   )
   > &RigUnit_ParentConstraint(
    AdvancedSettings=(RotationOrderForFilter=XZY)
    Child=(Type=Bone,Name="jacket_c_rear_01")
    Parents=((Item=(Type=Bone,Name="spine_04"),Weight=0.0565),(Item=(Type=Bone,Name="pelvis"),Weight=0.1654),(Item=(Type=Bone,Name="spine_01"),Weight=0.1772),(Item=(Type=Bone,Name="spine_03"),Weight=0.295),(Item=(Type=Bone,Name="spine_02"),Weight=0.3059))
    Weight=1
   )
   > &RigUnit_ParentConstraint(
    AdvancedSettings=(RotationOrderForFilter=XZY)
    Child=(Type=Bone,Name="jacket_r_rear_01")
    Parents="((Item=(Type=Bone,Name=\"clavicle_r\"),Weight=0.0027),(Item=(Type=Bone,Name=\"spine_04\"),Weight=0.0499),(Item=(Type=Bone,Name=\"pelvis\"),Weight=0.1815),(Item=(Type=Bone,Name=\"spine_01\"),Weight=0.1874),(Item=(Type=Bone,Name=\"spine_03\"),Weight=0.2621),(Item=(Typ ...(truncated)"
    Weight=1
   )
   > &RigUnit_ParentConstraint(
    AdvancedSettings=(RotationOrderForFilter=XZY)
    Child=(Type=Bone,Name="jacket_r_side_01")
    Parents="((Item=(Type=Bone,Name=\"clavicle_r\"),Weight=0.0206),(Item=(Type=Bone,Name=\"spine_04\"),Weight=0.0459),(Item=(Type=Bone,Name=\"pelvis\"),Weight=0.1572),(Item=(Type=Bone,Name=\"spine_01\"),Weight=0.1906),(Item=(Type=Bone,Name=\"spine_03\"),Weight=0.2529),(Item=(Typ ...(truncated)"
    Weight=1
   )
   > &RigUnit_ParentConstraint(
    AdvancedSettings=(RotationOrderForFilter=XZY)
    Child=(Type=Bone,Name="dyn_sackbelt_l_clav_root")
    Parents="((Item=(Type=Bone,Name=\"spine_04\"),Weight=0.0082),(Item=(Type=Bone,Name=\"spine_03\"),Weight=0.0318),(Item=(Type=Bone,Name=\"upperarm_twist_02_l\"),Weight=0.0357),(Item=(Type=Bone,Name=\"deform_upperarm_fr_l\"),Weight=0.0581),(Item=(Type=Bone,Name=\"spine_05\"),We ...(truncated)"
    Weight=1
   )
   > RigVMFunctionReturnNode()
  ]}
 New Function_ContainedGraph {nodes [
   RigVMFunctionEntryNode()
   > SequenceAggregate = RigVMAggregateNode()
   n1 = &RigUnit_SphericalPoseReader(
    ActiveRegionScaleFactors=(PositiveWidth=1,NegativeWidth=1,PositiveHeight=1,NegativeHeight=1)
    ActiveRegionSize=0.1
    Debug=(bDrawDebug=false,DebugScale=25,DebugSegments=20,DebugThickness=0.25)
    DriverAxis=(X=1,Y=0,Z=0)
    DriverItem=(Type=Bone,Name="head")
    FalloffRegionScaleFactors=(PositiveWidth=1,NegativeWidth=1,PositiveHeight=1,NegativeHeight=1)
    FalloffSize=0.5
    OptionalParentItem=(Type=Bone,Name="spine_05")
    RotationOffset=(X=90,Y=0,Z=-25)
    ExecutePin=@SequenceAggregate.A
   )
   > n2 = &RigUnit_SphericalPoseReader(
    ActiveRegionScaleFactors=(PositiveWidth=1,NegativeWidth=1,PositiveHeight=1,NegativeHeight=1)
    ActiveRegionSize=0.1
    Debug=(bDrawDebug=false,DebugScale=25,DebugSegments=20,DebugThickness=0.25)
    DriverAxis=(X=1,Y=0,Z=0)
    DriverItem=(Type=Bone,Name="head")
    FalloffRegionScaleFactors=(PositiveWidth=1,NegativeWidth=1,PositiveHeight=1,NegativeHeight=1)
    FalloffSize=0.5
    OptionalParentItem=(Type=Bone,Name="spine_05")
    RotationOffset=(X=-90,Y=0,Z=20)
   )
   > &RigUnit_ModifyTransforms(
    ItemToModify="((Item=(Type=Bone,Name=\"deform_hoodie_up_sternum\"),Transform=(Rotation=(X=0,Y=0,Z=0,W=1),Translation=(X=20,Y=0,Z=0),Scale3D=(X=1,Y=1,Z=1))),(Item=(Type=Bone,Name=\"deform_hoodie_up_back\"),Transform=(Rotation=(X=0,Y=0,Z=0,W=1),Translation=(X=0,Y=0.5,Z=0),Sca ...(truncated)"
    Weight=@n2.OutputParam
   )
   > n3 = &RigUnit_SphericalPoseReader(
    ActiveRegionScaleFactors=(PositiveWidth=1,NegativeWidth=1,PositiveHeight=1,NegativeHeight=1)
    ActiveRegionSize=0.05
    Debug=(bDrawDebug=false,DebugScale=25,DebugSegments=20,DebugThickness=0.25)
    DriverAxis=(X=0,Y=1,Z=0)
    DriverItem=(Type=Bone,Name="head")
    FalloffRegionScaleFactors=(PositiveWidth=1,NegativeWidth=1,PositiveHeight=1,NegativeHeight=1)
    FalloffSize=0.3
    OptionalParentItem=(Type=Bone,Name="spine_05")
    RotationOffset=(X=20,Y=180,Z=0)
   )
   > &RigUnit_ModifyTransforms(
    ItemToModify="((Item=(Type=Bone,Name=\"deform_hoodie_up_back\"),Transform=(Rotation=(X=0,Y=0,Z=0,W=1),Translation=(X=0,Y=0,Z=0),Scale3D=(X=1,Y=1,Z=1))),(Item=(Type=Bone,Name=\"deform_hoodie_up_front\"),Transform=(Rotation=(X=0,Y=0.3827,Z=0,W=0.9239),Translation=(X=0,Y=0,Z=- ...(truncated)"
    ItemToModify.2.Transform.Rotation=&RigVMFunction_MathQuaternionFromEuler(Euler=(X=0,Y=45,Z=0)).Result
    Weight=@n3.OutputParam
   )
   > n4 = &RigUnit_SphericalPoseReader(
    ActiveRegionScaleFactors=(PositiveWidth=1,NegativeWidth=1,PositiveHeight=1,NegativeHeight=1)
    ActiveRegionSize=0.05
    Debug=(bDrawDebug=false,DebugScale=25,DebugSegments=20,DebugThickness=0.25)
    DriverAxis=(X=0,Y=1,Z=0)
    DriverItem=(Type=Bone,Name="head")
    FalloffRegionScaleFactors=(PositiveWidth=1,NegativeWidth=1,PositiveHeight=1,NegativeHeight=1)
    FalloffSize=0.3
    OptionalParentItem=(Type=Bone,Name="spine_05")
    RotationOffset=(X=20,Y=0,Z=0)
   )
   RigVMFunction_MathQuaternionFromEuler = &RigVMFunction_MathQuaternionFromEuler(Euler=(X=0,Y=-45,Z=0))
   > &RigUnit_ModifyTransforms(
    ItemToModify="((Item=(Type=Bone,Name=\"deform_hoodie_up_back\"),Transform=(Rotation=(X=0,Y=0,Z=0,W=1),Translation=(X=0,Y=0,Z=0),Scale3D=(X=1,Y=1,Z=1))),(Item=(Type=Bone,Name=\"deform_hoodie_up_front\"),Transform=(Rotation=(X=0,Y=-0.3827,Z=0,W=0.9239),Translation=(X=0,Y=0,Z= ...(truncated)"
    ItemToModify.2.Transform.Rotation=@RigVMFunction_MathQuaternionFromEuler.Result
    Weight=@n4.OutputParam
   )
   Remap = &RigVMFunction_MathDoubleRemap(SourceMaximum=1, SourceMinimum=0, TargetMaximum=0, TargetMinimum=1, bClamp=False, Value=@n1.OutputParam)
   &RigUnit_ParentConstraint(
    AdvancedSettings=(InterpolationType=Shortest,RotationOrderForFilter=XZY)
    Child=(Type=Bone,Name="deform_hoodie_up_front_root")
    Parents=((Item=(Type=Bone,Name="head"),Weight=0.2),(Item=(Type=Bone,Name="neck_01"),Weight=0.4),(Item=(Type=Bone,Name="neck_02"),Weight=0.4))
    bMaintainOffset=true
    Weight=@Remap.Result
    ExecutePin=@SequenceAggregate.B
   )
   > &RigUnit_ParentConstraint(
    AdvancedSettings=(RotationOrderForFilter=XZY)
    Child=(Type=Bone,Name="deform_hoodie_up_front_edge")
    Parents=((Item=(Type=Bone,Name="head"),Weight=1))
    Weight=1
   )
   ShoulderBelt = RigVMFunctionReferenceNode(
    arm_lowered_bag=(Rotation=(X=0,Y=0,Z=0,W=1),Translation=(X=-4,Y=0.5,Z=0),Scale3D=(X=1,Y=1,Z=1))
    arm_lowered_clav=(Rotation=(X=0,Y=0,Z=0,W=1),Translation=(X=0,Y=-1,Z=1),Scale3D=(X=1,Y=1.25,Z=1))
    arm_lowered_hip=(Rotation=(X=0,Y=0,Z=0,W=1),Translation=(X=2,Y=-2,Z=0),Scale3D=(X=1,Y=1,Z=1))
    arm_lowered_pec=(Rotation=(X=0,Y=0,Z=0,W=1),Translation=(X=0,Y=0,Z=1),Scale3D=(X=1,Y=1,Z=1))
    arm_lowered_scap=(Rotation=(X=0,Y=0,Z=0,W=1),Translation=(X=0,Y=0,Z=0),Scale3D=(X=1,Y=1,Z=1))
    arm_raised_bag=(Rotation=(X=0,Y=0,Z=0,W=1),Translation=(X=-2,Y=0.5,Z=3),Scale3D=(X=1,Y=1,Z=1))
    arm_raised_clav=(Rotation=(X=0,Y=0,Z=0,W=1),Translation=(X=0,Y=-0.5,Z=0.5),Scale3D=(X=1,Y=1.1,Z=1))
    arm_raised_hip=(Rotation=(X=0,Y=0,Z=0,W=1),Translation=(X=0,Y=0,Z=0),Scale3D=(X=1,Y=1,Z=1))
    arm_raised_pec=(Rotation=(X=0,Y=0,Z=0,W=1),Translation=(X=0,Y=-4,Z=1),Scale3D=(X=1,Y=1,Z=1))
    arm_raised_scap=(Rotation=(X=0,Y=0,Z=0,W=1),Translation=(X=-0.5,Y=0,Z=2),Scale3D=(X=1,Y=1,Z=1))
    clav_raised_clav=(Rotation=(X=0,Y=0,Z=0,W=1),Translation=(X=-0.5,Y=0,Z=0),Scale3D=(X=1,Y=1,Z=1))
    clav_raised_pec=(Rotation=(X=0,Y=0,Z=0,W=1),Translation=(X=0,Y=0,Z=0),Scale3D=(X=1,Y=1,Z=1))
    clav_raised_scap=(Rotation=(X=0,Y=0,Z=0,W=1),Translation=(X=0,Y=0,Z=0),Scale3D=(X=1,Y=1,Z=1))
    clav_retracted_clav=(Rotation=(X=0,Y=0,Z=0,W=1),Translation=(X=0,Y=1,Z=0),Scale3D=(X=1,Y=1,Z=1))
    clav_retracted_scap=(Rotation=(X=0,Y=0,Z=0,W=1),Translation=(X=0,Y=0.5,Z=0),Scale3D=(X=1,Y=1,Z=1))
    arm_raised_pec.Rotation=&RigVMFunction_MathQuaternionFromEuler(Euler=(X=10,Y=-5,Z=20)).Result
    arm_lowered_pec.Rotation=&RigVMFunction_MathQuaternionFromEuler(Euler=(X=0,Y=0,Z=10)).Result
    arm_lowered_bag.Rotation=&RigVMFunction_MathQuaternionFromEuler(Euler=(X=0,Y=0,Z=15)).Result
    arm_lowered_hip.Rotation=&RigVMFunction_MathQuaternionFromEuler(Euler=(X=-12,Y=15,Z=0)).Result
    arm_raised_bag.Rotation=&RigVMFunction_MathQuaternionFromEuler(Euler=(X=0,Y=-10,Z=5)).Result
    arm_raised_hip.Rotation=&RigVMFunction_MathQuaternionFromEuler(Euler=(X=0,Y=0,Z=0)).Result
    arm_raised_clav.Rotation=&RigVMFunction_MathQuaternionFromEuler(Euler=(X=10,Y=15,Z=0)).Result
    arm_raised_scap.Rotation=&RigVMFunction_MathQuaternionFromEuler(Euler=(X=0,Y=-10,Z=0)).Result
    arm_lowered_clav.Rotation=&RigVMFunction_MathQuaternionFromEuler(Euler=(X=0,Y=0,Z=0)).Result
    arm_lowered_scap.Rotation=&RigVMFunction_MathQuaternionFromEuler(Euler=(X=0,Y=0,Z=0)).Result
    clav_retracted_scap.Rotation=&RigVMFunction_MathQuaternionFromEuler(Euler=(X=10,Y=0,Z=5)).Result
    clav_retracted_clav.Rotation=&RigVMFunction_MathQuaternionFromEuler(Euler=(X=10,Y=0,Z=10)).Result
    clav_raised_clav.Rotation=&RigVMFunction_MathQuaternionFromEuler(Euler=(X=0,Y=10,Z=0)).Result
    clav_raised_scap.Rotation=&RigVMFunction_MathQuaternionFromEuler(Euler=(X=0,Y=0,Z=0)).Result
    clav_raised_pec.Rotation=&RigVMFunction_MathQuaternionFromEuler(Euler=(X=0,Y=0,Z=-10)).Result
    ExecuteContext=@SequenceAggregate.C
   )
   RigVMFunctionReturnNode(ExecuteContext=@SequenceAggregate.D)
  ]}
 New Function_ContainedGraph {nodes [
   n1 = RigVMFunctionEntryNode()
   > SequenceAggregate = RigVMAggregateNode()
   n2 = &RigUnit_SphericalPoseReader(
    ActiveRegionScaleFactors=(PositiveWidth=1,NegativeWidth=1,PositiveHeight=1,NegativeHeight=1)
    ActiveRegionSize=0.1
    Debug=(bDrawDebug=false,DebugScale=25,DebugSegments=20,DebugThickness=0.25)
    DriverAxis=(X=1,Y=0,Z=0)
    DriverItem=(Type=Bone,Name="upperarm_l")
    FalloffRegionScaleFactors=(PositiveWidth=1,NegativeWidth=1,PositiveHeight=1,NegativeHeight=1)
    FalloffSize=0.5
    OptionalParentItem=(Type=Bone,Name="spine_05")
    RotationOffset=(X=60,Y=0,Z=20)
    ExecutePin=@SequenceAggregate.A
   )
   > &RigUnit_ModifyTransforms(
    ItemToModify="((Item=(Type=Bone,Name=\"dyn_sackbelt_l_pec\"),Transform=(Rotation=(X=0,Y=0,Z=0,W=1),Translation=(X=0,Y=-2,Z=1),Scale3D=(X=1,Y=1,Z=1))),(Item=(Type=Bone,Name=\"dyn_sackbag\"),Transform=(Rotation=(X=0,Y=0,Z=0,W=1),Translation=(X=-2,Y=0,Z=3),Scale3D=(X=1,Y=1,Z=1 ...(truncated)"
    ItemToModify.1.Transform=@n1.arm_raised_bag
    ItemToModify.3.Transform=@n1.arm_raised_clav
    ItemToModify.2.Transform=@n1.arm_raised_hip
    ItemToModify.0.Transform=@n1.arm_raised_pec
    ItemToModify.4.Transform=@n1.arm_raised_scap
    Weight=@n2.OutputParam
   )
   n3 = &RigUnit_SphericalPoseReader(
    ActiveRegionScaleFactors=(PositiveWidth=1,NegativeWidth=1,PositiveHeight=1,NegativeHeight=1)
    ActiveRegionSize=0.1
    Debug=(bDrawDebug=false,DebugScale=25,DebugSegments=20,DebugThickness=0.25)
    DriverAxis=(X=1,Y=0,Z=0)
    DriverItem=(Type=Bone,Name="upperarm_l")
    FalloffRegionScaleFactors=(PositiveWidth=1,NegativeWidth=1,PositiveHeight=1,NegativeHeight=1)
    FalloffSize=0.5
    OptionalParentItem=(Type=Bone,Name="spine_05")
    RotationOffset=(X=150,Y=0,Z=-60)
    ExecutePin=@SequenceAggregate.B
   )
   > &RigUnit_ModifyTransforms(
    ItemToModify="((Item=(Type=Bone,Name=\"dyn_sackbelt_l_pec\"),Transform=(Rotation=(X=0,Y=0,Z=0,W=1),Translation=(X=0,Y=1,Z=1),Scale3D=(X=1,Y=1,Z=1))),(Item=(Type=Bone,Name=\"dyn_sackbag\"),Transform=(Rotation=(X=0,Y=0,Z=0,W=1),Translation=(X=-4,Y=0,Z=0),Scale3D=(X=1,Y=1,Z=1) ...(truncated)"
    ItemToModify.1.Transform=@n1.arm_lowered_bag
    ItemToModify.3.Transform=@n1.arm_lowered_clav
    ItemToModify.2.Transform=@n1.arm_lowered_hip
    ItemToModify.0.Transform=@n1.arm_lowered_pec
    ItemToModify.4.Transform=@n1.arm_lowered_scap
    Weight=@n3.OutputParam
   )
   n4 = &RigUnit_SphericalPoseReader(
    ActiveRegionScaleFactors=(PositiveWidth=1,NegativeWidth=1,PositiveHeight=1,NegativeHeight=1)
    ActiveRegionSize=0.1
    Debug=(bDrawDebug=false,DebugScale=25,DebugSegments=20,DebugThickness=0.25)
    DriverAxis=(X=1,Y=0,Z=0)
    DriverItem=(Type=Bone,Name="clavicle_l")
    FalloffRegionScaleFactors=(PositiveWidth=1,NegativeWidth=1,PositiveHeight=1,NegativeHeight=1)
    FalloffSize=0.5
    OptionalParentItem=(Type=Bone,Name="spine_05")
    RotationOffset=(X=-90,Y=0,Z=45)
    ExecutePin=@SequenceAggregate.C
   )
   > &RigUnit_ModifyTransforms(
    ItemToModify="((Item=(Type=Bone,Name=\"dyn_sackbelt_l_scap\"),Transform=(Rotation=(X=0,Y=0,Z=0,W=1),Translation=(X=0,Y=0.5,Z=0),Scale3D=(X=1,Y=1,Z=1))),(Item=(Type=Bone,Name=\"dyn_sackbelt_l_clav\"),Transform=(Rotation=(X=0,Y=0,Z=0,W=1),Translation=(X=0,Y=0,Z=0),Scale3D=(X= ...(truncated)"
    ItemToModify.1.Transform=@n1.clav_retracted_clav
    ItemToModify.0.Transform=@n1.clav_retracted_scap
    Weight=@n4.OutputParam
   )
   n5 = &RigUnit_SphericalPoseReader(
    ActiveRegionScaleFactors=(PositiveWidth=1,NegativeWidth=1,PositiveHeight=1,NegativeHeight=1)
    ActiveRegionSize=0.1
    Debug=(bDrawDebug=false,DebugScale=25,DebugSegments=20,DebugThickness=0.25)
    DriverAxis=(X=1,Y=0,Z=0)
    DriverItem=(Type=Bone,Name="clavicle_l")
    FalloffRegionScaleFactors=(PositiveWidth=1,NegativeWidth=1,PositiveHeight=1,NegativeHeight=1)
    FalloffSize=0.5
    OptionalParentItem=(Type=Bone,Name="spine_05")
    RotationOffset=(X=0,Y=-45,Z=0)
    ExecutePin=@SequenceAggregate.D
   )
   > &RigUnit_ModifyTransforms(
    ItemToModify="((Item=(Type=Bone,Name=\"dyn_sackbelt_l_scap\"),Transform=(Rotation=(X=0,Y=0,Z=0,W=1),Translation=(X=0,Y=0,Z=0),Scale3D=(X=1,Y=1,Z=1))),(Item=(Type=Bone,Name=\"dyn_sackbelt_l_clav\"),Transform=(Rotation=(X=0,Y=0,Z=0,W=1),Translation=(X=-0.5,Y=0,Z=0),Scale3D=(X ...(truncated)"
    ItemToModify.1.Transform=@n1.clav_raised_clav
    ItemToModify.2.Transform=@n1.clav_raised_pec
    ItemToModify.0.Transform=@n1.clav_raised_scap
    Weight=@n5.OutputParam
   )
   RigVMFunctionReturnNode(ExecuteContext=@SequenceAggregate.E)
   &RigVMFunction_MathQuaternionFromEuler(Euler=(X=20,Y=-5,Z=0))
   &RigVMFunction_MathQuaternionFromEuler(Euler=(X=0,Y=-5,Z=0))
   &RigVMFunction_MathQuaternionFromEuler(Euler=(X=0,Y=-10,Z=15))
   &RigVMFunction_MathQuaternionFromEuler(Euler=(X=-12,Y=15,Z=0))
   &RigVMFunction_MathQuaternionFromEuler(Euler=(X=0,Y=-10,Z=0))
   &RigVMFunction_MathQuaternionFromEuler(Euler=(X=0,Y=0,Z=0))
   &RigVMFunction_MathQuaternionFromEuler(Euler=(X=10,Y=20,Z=0))
   &RigVMFunction_MathQuaternionFromEuler(Euler=(X=10,Y=-10,Z=0))
   &RigVMFunction_MathQuaternionFromEuler(Euler=(X=10,Y=20,Z=0))
   &RigVMFunction_MathQuaternionFromEuler(Euler=(X=0,Y=0,Z=0))
   &RigVMFunction_MathQuaternionFromEuler(Euler=(X=10,Y=0,Z=0))
   &RigVMFunction_MathQuaternionFromEuler(Euler=(X=10,Y=0,Z=10))
   &RigVMFunction_MathQuaternionFromEuler(Euler=(X=0,Y=10,Z=0))
   &RigVMFunction_MathQuaternionFromEuler(Euler=(X=0,Y=0,Z=0))
   &RigVMFunction_MathQuaternionFromEuler(Euler=(X=0,Y=0,Z=-10))
  ]}
 New Function_ContainedGraph {nodes [
   RigVMFunctionEntryNode()
   > ShoulderBelt = RigVMFunctionReferenceNode(
    arm_lowered_bag=(Rotation=(X=0,Y=0,Z=0,W=1),Translation=(X=-4,Y=0.5,Z=0),Scale3D=(X=1,Y=1,Z=1))
    arm_lowered_clav=(Rotation=(X=0,Y=0,Z=0,W=1),Translation=(X=0.5,Y=0,Z=0.5),Scale3D=(X=1,Y=1,Z=1))
    arm_lowered_hip=(Rotation=(X=0,Y=0,Z=0,W=1),Translation=(X=2,Y=-2,Z=0),Scale3D=(X=1,Y=1,Z=1))
    arm_lowered_pec=(Rotation=(X=0,Y=0,Z=0,W=1),Translation=(X=0,Y=1,Z=1),Scale3D=(X=1,Y=1,Z=1))
    arm_lowered_scap=(Rotation=(X=0,Y=0,Z=0,W=1),Translation=(X=0,Y=0,Z=0),Scale3D=(X=1,Y=1,Z=1))
    arm_raised_bag=(Rotation=(X=0,Y=0,Z=0,W=1),Translation=(X=-2,Y=0.5,Z=3),Scale3D=(X=1,Y=1,Z=1))
    arm_raised_clav=(Rotation=(X=0,Y=0,Z=0,W=1),Translation=(X=0,Y=-1,Z=1),Scale3D=(X=1,Y=1,Z=1))
    arm_raised_hip=(Rotation=(X=0,Y=0,Z=0,W=1),Translation=(X=0,Y=0,Z=0),Scale3D=(X=1,Y=1,Z=1))
    arm_raised_pec=(Rotation=(X=0,Y=0,Z=0,W=1),Translation=(X=0.5,Y=-3.5,Z=3.5),Scale3D=(X=1,Y=1,Z=1))
    arm_raised_scap=(Rotation=(X=0,Y=0,Z=0,W=1),Translation=(X=-0.5,Y=0,Z=1),Scale3D=(X=1,Y=1,Z=1))
    clav_raised_clav=(Rotation=(X=0,Y=0,Z=0,W=1),Translation=(X=-0.5,Y=0,Z=0),Scale3D=(X=1,Y=1,Z=1))
    clav_raised_pec=(Rotation=(X=0,Y=0,Z=0,W=1),Translation=(X=0,Y=0,Z=0),Scale3D=(X=1,Y=1,Z=1))
    clav_raised_scap=(Rotation=(X=0,Y=0,Z=0,W=1),Translation=(X=0,Y=0,Z=0),Scale3D=(X=1,Y=1,Z=1))
    clav_retracted_clav=(Rotation=(X=0,Y=0,Z=0,W=1),Translation=(X=0,Y=0,Z=0),Scale3D=(X=1,Y=1,Z=1))
    clav_retracted_scap=(Rotation=(X=0,Y=0,Z=0,W=1),Translation=(X=0,Y=0.5,Z=0),Scale3D=(X=1,Y=1,Z=1))
    arm_raised_pec.Rotation=&RigVMFunction_MathQuaternionFromEuler(Euler=(X=-40,Y=5,Z=20)).Result
    arm_lowered_pec.Rotation=&RigVMFunction_MathQuaternionFromEuler(Euler=(X=0,Y=-5,Z=10)).Result
    arm_lowered_bag.Rotation=&RigVMFunction_MathQuaternionFromEuler(Euler=(X=0,Y=0,Z=15)).Result
    arm_lowered_hip.Rotation=&RigVMFunction_MathQuaternionFromEuler(Euler=(X=-12,Y=15,Z=0)).Result
    arm_raised_bag.Rotation=&RigVMFunction_MathQuaternionFromEuler(Euler=(X=0,Y=-10,Z=5)).Result
    arm_raised_hip.Rotation=&RigVMFunction_MathQuaternionFromEuler(Euler=(X=0,Y=0,Z=0)).Result
    arm_raised_clav.Rotation=&RigVMFunction_MathQuaternionFromEuler(Euler=(X=10,Y=5,Z=0)).Result
    arm_raised_scap.Rotation=&RigVMFunction_MathQuaternionFromEuler(Euler=(X=10,Y=-10,Z=0)).Result
    arm_lowered_clav.Rotation=&RigVMFunction_MathQuaternionFromEuler(Euler=(X=0,Y=0,Z=0)).Result
    arm_lowered_scap.Rotation=&RigVMFunction_MathQuaternionFromEuler(Euler=(X=0,Y=0,Z=0)).Result
    clav_retracted_scap.Rotation=&RigVMFunction_MathQuaternionFromEuler(Euler=(X=10,Y=10,Z=0)).Result
    clav_retracted_clav.Rotation=&RigVMFunction_MathQuaternionFromEuler(Euler=(X=10,Y=0,Z=10)).Result
    clav_raised_clav.Rotation=&RigVMFunction_MathQuaternionFromEuler(Euler=(X=0,Y=10,Z=0)).Result
    clav_raised_scap.Rotation=&RigVMFunction_MathQuaternionFromEuler(Euler=(X=0,Y=0,Z=0)).Result
    clav_raised_pec.Rotation=&RigVMFunction_MathQuaternionFromEuler(Euler=(X=0,Y=0,Z=-10)).Result
   )
   > RigVMFunctionReturnNode()
  ]}
 RigVMFunctionLibrary {nodes [
   Belt = RigVMCollapseNode(ExecuteContext=(), thigh_r_alpha=0)
   Deform_GetMirrorTransform_FNC = RigVMCollapseNode(ExecuteContext=(), transform=(Rotation=(X=0,Y=0,Z=0,W=1),Translation=(X=0,Y=0,Z=0),Scale3D=(X=1,Y=1,Z=1)))
   HoodieDown = RigVMCollapseNode(ExecuteContext=())
   HoodieUp = RigVMCollapseNode(ExecuteContext=())
   Jacket = RigVMCollapseNode(ExecuteContext=(), thigh_l_alpha=0, thigh_r_alpha=0)
   Rivets = RigVMCollapseNode(ExecuteContext=())
   ShoulderBelt = RigVMCollapseNode(
    ExecuteContext=()
    arm_lowered_bag=(Rotation=(X=0,Y=0,Z=0,W=1),Translation=(X=0,Y=0,Z=0),Scale3D=(X=1,Y=1,Z=1))
    arm_lowered_clav=(Rotation=(X=0,Y=0,Z=0,W=1),Translation=(X=0,Y=0,Z=0),Scale3D=(X=1,Y=1,Z=1))
    arm_lowered_hip=(Rotation=(X=0,Y=0,Z=0,W=1),Translation=(X=0,Y=0,Z=0),Scale3D=(X=1,Y=1,Z=1))
    arm_lowered_pec=(Rotation=(X=0,Y=0,Z=0,W=1),Translation=(X=0,Y=0,Z=0),Scale3D=(X=1,Y=1,Z=1))
    arm_lowered_scap=(Rotation=(X=0,Y=0,Z=0,W=1),Translation=(X=0,Y=0,Z=0),Scale3D=(X=1,Y=1,Z=1))
    arm_raised_bag=(Rotation=(X=0,Y=0,Z=0,W=1),Translation=(X=0,Y=0,Z=0),Scale3D=(X=1,Y=1,Z=1))
    arm_raised_clav=(Rotation=(X=0,Y=0,Z=0,W=1),Translation=(X=0,Y=0,Z=0),Scale3D=(X=1,Y=1,Z=1))
    arm_raised_hip=(Rotation=(X=0,Y=0,Z=0,W=1),Translation=(X=0,Y=0,Z=0),Scale3D=(X=1,Y=1,Z=1))
    arm_raised_pec=(Rotation=(X=0,Y=0,Z=0,W=1),Translation=(X=0,Y=0,Z=0),Scale3D=(X=1,Y=1,Z=1))
    arm_raised_scap=(Rotation=(X=0,Y=0,Z=0,W=1),Translation=(X=0,Y=0,Z=0),Scale3D=(X=1,Y=1,Z=1))
    clav_raised_clav=(Rotation=(X=0,Y=0,Z=0,W=1),Translation=(X=0,Y=0,Z=0),Scale3D=(X=1,Y=1,Z=1))
    clav_raised_pec=(Rotation=(X=0,Y=0,Z=0,W=1),Translation=(X=0,Y=0,Z=0),Scale3D=(X=1,Y=1,Z=1))
    clav_raised_scap=(Rotation=(X=0,Y=0,Z=0,W=1),Translation=(X=0,Y=0,Z=0),Scale3D=(X=1,Y=1,Z=1))
    clav_retracted_clav=(Rotation=(X=0,Y=0,Z=0,W=1),Translation=(X=0,Y=0,Z=0),Scale3D=(X=1,Y=1,Z=1))
    clav_retracted_scap=(Rotation=(X=0,Y=0,Z=0,W=1),Translation=(X=0,Y=0,Z=0),Scale3D=(X=1,Y=1,Z=1))
   )
  ]}
 RigVMModel {nodes [
   RigUnit_BeginExecution = &RigUnit_BeginExecution()
   Rivets = RigVMFunctionReferenceNode(ExecuteContext=@RigUnit_BeginExecution.ExecutePin)
   > &RigUnit_ModifyTransforms(
    ItemToModify=((Item=(Type=Bone,Name="dyn_sackbag_root"),Transform=(Rotation=(X=0,Y=0,Z=0,W=1),Translation=(X=0,Y=1,Z=0),Scale3D=(X=1,Y=1,Z=1))))
    Weight=1
   )
   > SequenceAggregate = RigVMAggregateNode()
   n1 = &RigVMFunction_ControlFlowBranch(ExecuteContext=@SequenceAggregate.A, Condition=RigVMVariableNode() {variable=IsCrouching}.Value)
   SequenceAggregate_1 = RigVMAggregateNode(ExecuteContext=@n1.False)
   n2 = &RigUnit_SphericalPoseReader(
    ActiveRegionScaleFactors=(PositiveWidth=2,NegativeWidth=2,PositiveHeight=1.5,NegativeHeight=1.5)
    ActiveRegionSize=0.1
    Debug=(bDrawDebug=false,DebugScale=25,DebugSegments=20,DebugThickness=0.25)
    DriverAxis=(X=-1,Y=0,Z=0)
    DriverItem=(Type=Bone,Name="thigh_l")
    FalloffRegionScaleFactors=(PositiveWidth=1,NegativeWidth=1,PositiveHeight=1,NegativeHeight=1)
    FalloffSize=0.6
    OptionalParentItem=(Type=Bone,Name="pelvis")
    RotationOffset=(X=90,Y=0,Z=0)
    ExecutePin=@SequenceAggregate_1.A
   )
   > n3 = &RigUnit_SphericalPoseReader(
    ActiveRegionScaleFactors=(PositiveWidth=1,NegativeWidth=2,PositiveHeight=1.5,NegativeHeight=1.5)
    ActiveRegionSize=0.1
    Debug=(bDrawDebug=false,DebugScale=25,DebugSegments=20,DebugThickness=0.25)
    DriverAxis=(X=1,Y=0,Z=0)
    DriverItem=(Type=Bone,Name="thigh_r")
    FalloffRegionScaleFactors=(PositiveWidth=1,NegativeWidth=1,PositiveHeight=1,NegativeHeight=1)
    FalloffSize=0.5
    OptionalParentItem=(Type=Bone,Name="pelvis")
    RotationOffset=(X=-90,Y=0,Z=0)
   )
   n4 = &RigVMFunction_MathQuaternionFromEuler(Euler=(X=30,Y=-45,Z=-20))
   > &RigUnit_ModifyTransforms(
    ItemToModify="((Item=(Type=Bone,Name=\"jacket_l_front_01_root\"),Transform=(Rotation=(X=0,Y=0,Z=0,W=1),Translation=(X=0,Y=0,Z=0),Scale3D=(X=1,Y=1,Z=1))),(Item=(Type=Bone,Name=\"jacket_l_side_01_root\"),Transform=(Rotation=(X=0,Y=0,Z=0,W=1),Translation=(X=0,Y=0,Z=0),Scale3D= ...(truncated)"
    ItemToModify.0.Transform.Rotation=@n4.Result
    ItemToModify.1.Transform.Rotation=&RigVMFunction_MathQuaternionFromEuler(Euler=(X=0,Y=-20,Z=0)).Result
    Weight=@n2.OutputParam
   )
   > Deform_GetMirrorTransform_FNC_2_2 = RigVMFunctionReferenceNode(transform=(Rotation=(X=0,Y=0,Z=0,W=1),Translation=(X=0,Y=0,Z=0),Scale3D=(X=1,Y=1,Z=1)), transform.Rotation=@n4.Result)
   > Deform_GetMirrorTransform_FNC_2_1_1 = RigVMFunctionReferenceNode(transform=(Rotation=(X=0,Y=0,Z=0,W=1),Translation=(X=0,Y=0,Z=0),Scale3D=(X=1,Y=1,Z=1)))
   > &RigUnit_ModifyTransforms(
    ItemToModify="((Item=(Type=Bone,Name=\"jacket_r_front_01_root\"),Transform=(Rotation=(X=0,Y=0,Z=0,W=1),Translation=(X=0,Y=0,Z=0),Scale3D=(X=1,Y=1,Z=1))),(Item=(Type=Bone,Name=\"jacket_r_side_01_root\"),Transform=(Rotation=(X=0,Y=0,Z=0,W=1),Translation=(X=0,Y=0,Z=0),Scale3D= ...(truncated)"
    ItemToModify.1.Transform=@Deform_GetMirrorTransform_FNC_2_1_1.mirrorTransform
    ItemToModify.0.Transform=@Deform_GetMirrorTransform_FNC_2_2.mirrorTransform
    Weight=@n3.OutputParam
   )
   > &RigUnit_ModifyTransforms(
    ItemToModify="((Item=(Type=Bone,Name=\"dyn_sackbag_root\"),Transform=(Rotation=(X=0,Y=0,Z=0,W=1),Translation=(X=0,Y=0,Z=0),Scale3D=(X=1,Y=1,Z=1))),(Item=(Type=Bone,Name=\"dyn_sackbelt_r_hip_root\"),Transform=(Rotation=(X=0,Y=0,Z=0,W=1),Translation=(X=0,Y=0,Z=0),Scale3D=(X=1 ...(truncated)"
    ItemToModify.0.Transform=RigVMVariableNode() {variable=sackbag_thigh_r_fwd_normal}.Value
    ItemToModify.1.Transform=RigVMVariableNode() {variable=sackbelt_hip_thigh_r_fwd_normal}.Value
    ItemToModify.2.Transform=RigVMVariableNode() {variable=sackbelt_lowerback_thigh_r_fwd_normal}.Value
    Weight=@n3.OutputParam
   )
   n5 = &RigUnit_SphericalPoseReader(
    ActiveRegionScaleFactors=(PositiveWidth=2,NegativeWidth=2,PositiveHeight=2,NegativeHeight=2)
    ActiveRegionSize=0.1
    Debug=(bDrawDebug=true,DebugScale=25,DebugSegments=20,DebugThickness=0.25)
    DriverAxis=(X=-1,Y=0,Z=0)
    DriverItem=(Type=Bone,Name="thigh_l")
    FalloffRegionScaleFactors=(PositiveWidth=1,NegativeWidth=1,PositiveHeight=1,NegativeHeight=1)
    FalloffSize=0.6
    OptionalParentItem=(Type=Bone,Name="pelvis")
    RotationOffset=(X=120,Y=0,Z=0)
    ExecutePin=@SequenceAggregate_1.B
   )
   > n6 = &RigUnit_SphericalPoseReader(
    ActiveRegionScaleFactors=(PositiveWidth=2,NegativeWidth=2,PositiveHeight=2,NegativeHeight=2)
    ActiveRegionSize=0.12
    Debug=(bDrawDebug=true,DebugScale=25,DebugSegments=20,DebugThickness=0.25)
    DriverAxis=(X=1,Y=0,Z=0)
    DriverItem=(Type=Bone,Name="thigh_r")
    FalloffRegionScaleFactors=(PositiveWidth=1,NegativeWidth=1,PositiveHeight=1,NegativeHeight=1)
    FalloffSize=0.6
    OptionalParentItem=(Type=Bone,Name="pelvis")
    RotationOffset=(X=-40,Y=-30,Z=0)
   )
   > &RigUnit_ModifyTransforms(
    ItemToModify="((Item=(Type=Bone,Name=\"jacket_l_side_01_root\"),Transform=(Rotation=(X=0,Y=0,Z=0,W=1),Translation=(X=0,Y=0,Z=0),Scale3D=(X=1,Y=1,Z=1))),(Item=(Type=Bone,Name=\"jacket_l_front_01_root\"),Transform=(Rotation=(X=0,Y=0,Z=0,W=1),Translation=(X=0,Y=0,Z=0),Scale3D= ...(truncated)"
    ItemToModify.0.Transform.Rotation=&RigVMFunction_MathQuaternionFromEuler(Euler=(X=0,Y=-20,Z=0)).Result
    ItemToModify.1.Transform.Rotation=&RigVMFunction_MathQuaternionFromEuler(Euler=(X=0,Y=-20,Z=0)).Result
    Weight=@n5.OutputParam
   )
   > &RigUnit_ModifyTransforms(
    ItemToModify="((Item=(Type=Bone,Name=\"jacket_r_side_01_root\"),Transform=(Rotation=(X=0,Y=0,Z=0,W=1),Translation=(X=0,Y=0,Z=0),Scale3D=(X=1,Y=1,Z=1))),(Item=(Type=Bone,Name=\"jacket_r_front_01_root\"),Transform=(Rotation=(X=0,Y=0,Z=0,W=1),Translation=(X=0,Y=0,Z=0),Scale3D= ...(truncated)"
    ItemToModify.0.Transform.Rotation=&RigVMFunction_MathQuaternionFromEuler(Euler=(X=0,Y=-20,Z=0)).Result
    ItemToModify.1.Transform.Rotation=&RigVMFunction_MathQuaternionFromEuler(Euler=(X=0,Y=-20,Z=0)).Result
    Weight=@n6.OutputParam
   )
   &RigUnit_ModifyTransforms(
    ItemToModify="((Item=(Type=Bone,Name=\"dyn_sackbag_root\"),Transform=(Rotation=(X=0,Y=0,Z=0,W=1),Translation=(X=0,Y=0,Z=2),Scale3D=(X=1,Y=1,Z=1))),(Item=(Type=Bone,Name=\"dyn_sackbelt_r_hip\"),Transform=(Rotation=(X=0,Y=0,Z=0,W=1),Translation=(X=0,Y=0,Z=2),Scale3D=(X=1,Y=1, ...(truncated)"
    Weight=1
    ItemToModify.1.Transform.Rotation=&RigVMFunction_MathQuaternionFromEuler(Euler=(X=10,Y=0,Z=0)).Result
    ItemToModify.0.Transform.Rotation=&RigVMFunction_MathQuaternionFromEuler(Euler=(X=0,Y=5,Z=0)).Result
    ExecutePin=@SequenceAggregate_1.C
   )
   RigVMFunction_MathQuaternionFromEuler = &RigVMFunction_MathQuaternionFromEuler(Euler=(X=0,Y=0,Z=0))
   &RigUnit_ModifyTransforms(
    ItemToModify="((Item=(Type=Bone,Name=\"dyn_sackbag_root\"),Transform=(Rotation=(X=0,Y=0,Z=0,W=1),Translation=(X=0,Y=0,Z=4),Scale3D=(X=1,Y=1,Z=1))),(Item=(Type=Bone,Name=\"dyn_sackbelt_r_hip\"),Transform=(Rotation=(X=0,Y=0,Z=0,W=1),Translation=(X=0,Y=0,Z=1),Scale3D=(X=1,Y=1, ...(truncated)"
    Weight=1
    ExecutePin=@n1.True
    ItemToModify.0.Transform.Rotation=@RigVMFunction_MathQuaternionFromEuler.Result
    ItemToModify.1.Transform.Rotation=&RigVMFunction_MathQuaternionFromEuler(Euler=(X=10,Y=0,Z=0)).Result
   )
   > n7 = &RigUnit_SphericalPoseReader(
    ActiveRegionScaleFactors=(PositiveWidth=2,NegativeWidth=1,PositiveHeight=1,NegativeHeight=2)
    ActiveRegionSize=0.1
    Debug=(bDrawDebug=true,DebugScale=25,DebugSegments=20,DebugThickness=0.25)
    DriverAxis=(X=-1,Y=0,Z=0)
    DriverItem=(Type=Bone,Name="thigh_l")
    FalloffRegionScaleFactors=(PositiveWidth=1,NegativeWidth=1,PositiveHeight=1,NegativeHeight=1)
    FalloffSize=0.5
    OptionalParentItem=(Type=Bone,Name="spine_04")
    RotationOffset=(X=130,Y=0,Z=-70)
   )
   > n8 = &RigUnit_SphericalPoseReader(
    ActiveRegionScaleFactors=(PositiveWidth=0.5,NegativeWidth=1,PositiveHeight=1,NegativeHeight=1)
    ActiveRegionSize=0.1
    Debug=(bDrawDebug=false,DebugScale=25,DebugSegments=20,DebugThickness=0.25)
    DriverAxis=(X=1,Y=0,Z=0)
    DriverItem=(Type=Bone,Name="thigh_r")
    FalloffRegionScaleFactors=(PositiveWidth=1,NegativeWidth=1,PositiveHeight=1,NegativeHeight=1)
    FalloffSize=0.5
    OptionalParentItem=(Type=Bone,Name="spine_04")
    RotationOffset=(X=-50,Y=0,Z=-45)
   )
   > &RigUnit_ModifyTransforms(
    ItemToModify="((Item=(Type=Bone,Name=\"jacket_l_front_01_root\"),Transform=(Rotation=(X=0,Y=0,Z=0,W=1),Translation=(X=0,Y=-3,Z=0),Scale3D=(X=1,Y=1,Z=1))),(Item=(Type=Bone,Name=\"jacket_l_side_01_root\"),Transform=(Rotation=(X=0,Y=0,Z=0,W=1),Translation=(X=0,Y=0,Z=5),Scale3D ...(truncated)"
    Weight=@n7.OutputParam
    ItemToModify.0.Transform=RigVMVariableNode() {variable=jacket_front_leg_fwd}.Value
    ItemToModify.1.Transform=RigVMVariableNode() {variable=jacket_side_leg_fwd}.Value
   )
   > Deform_GetMirrorTransform_FNC_2 = RigVMFunctionReferenceNode(transform=RigVMVariableNode() {variable=jacket_front_leg_fwd}.Value)
   > Deform_GetMirrorTransform_FNC_2_1 = RigVMFunctionReferenceNode(transform=RigVMVariableNode() {variable=jacket_side_leg_fwd}.Value)
   > &RigUnit_ModifyTransforms(
    ItemToModify="((Item=(Type=Bone,Name=\"jacket_r_front_01_root\"),Transform=(Rotation=(X=0,Y=0,Z=0,W=1),Translation=(X=0,Y=3,Z=0),Scale3D=(X=1,Y=1,Z=1))),(Item=(Type=Bone,Name=\"jacket_r_side_01_root\"),Transform=(Rotation=(X=0,Y=0,Z=0,W=1),Translation=(X=2,Y=0,Z=0),Scale3D= ...(truncated)"
    WeightMaximum=0.75
    ItemToModify.0.Transform=@Deform_GetMirrorTransform_FNC_2.mirrorTransform
    Weight=@n8.OutputParam
   )
   > &RigUnit_ModifyTransforms(
    ItemToModify="((Item=(Type=Bone,Name=\"dyn_sackbag_root\"),Transform=(Rotation=(X=0,Y=0,Z=0,W=1),Translation=(X=4,Y=3,Z=2),Scale3D=(X=1,Y=1,Z=1))),(Item=(Type=Bone,Name=\"dyn_sackbelt_r_hip_root\"),Transform=(Rotation=(X=0,Y=0,Z=0,W=1),Translation=(X=0,Y=2,Z=5),Scale3D=(X=1 ...(truncated)"
    ItemToModify.0.Transform.Rotation=&RigVMFunction_MathQuaternionFromEuler(Euler=(X=10,Y=15,Z=-5)).Result
    ItemToModify.1.Transform.Rotation=&RigVMFunction_MathQuaternionFromEuler(Euler=(X=20,Y=0,Z=0)).Result
    Weight=@n8.OutputParam
   )
   n9 = &RigVMFunction_ControlFlowBranch(
    Condition=&RigVMFunction_MathBoolOr(A=RigVMVariableNode() {variable=baseball_sliding}.Value, B=RigVMVariableNode() {variable=knee_sliding}.Value).Result
    ExecuteContext=@SequenceAggregate.B
   )
   Cast = &RigVMFunction_MathBoolToFloat(Value=RigVMVariableNode() {variable=baseball_sliding}.Value)
   &RigUnit_ModifyTransforms(
    ItemToModify="((Item=(Type=Bone,Name=\"jacket_c_rear_01_root\"),Transform=(Rotation=(X=0,Y=0,Z=0,W=1),Translation=(X=0,Y=0,Z=0),Scale3D=(X=1,Y=1,Z=1))),(Item=(Type=Bone,Name=\"jacket_l_front_01_root\"),Transform=(Rotation=(X=0,Y=0,Z=0,W=1),Translation=(X=0,Y=0,Z=0),Scale3D= ...(truncated)"
    ExecutePin=@n9.True
    Weight=@Cast.Result
    ItemToModify.3.Transform=RigVMVariableNode() {variable=baseball_slide_jacket_r_front}.Value
    ItemToModify.4.Transform=RigVMVariableNode() {variable=baseball_slide_jacket_r_rear}.Value
    ItemToModify.0.Transform=RigVMVariableNode() {variable=baseball_slide_jacket_c_rear}.Value
    ItemToModify.1.Transform=RigVMVariableNode() {variable=baseball_slide_jacket_l_front}.Value
    ItemToModify.2.Transform=RigVMVariableNode() {variable=baseball_slide_jacket_l_rear}.Value
   )
   > &RigUnit_ModifyTransforms(
    ItemToModify="((Item=(Type=Bone,Name=\"jacket_c_rear_01_root\"),Transform=(Rotation=(X=0,Y=0,Z=0,W=1),Translation=(X=0,Y=0,Z=0),Scale3D=(X=1,Y=1,Z=1))),(Item=(Type=Bone,Name=\"jacket_l_front_01_root\"),Transform=(Rotation=(X=0,Y=0,Z=0,W=1),Translation=(X=0,Y=0,Z=0),Scale3D= ...(truncated)"
    Weight=&RigVMFunction_MathBoolToFloat(Value=RigVMVariableNode() {variable=knee_sliding}.Value).Result
    ItemToModify.0.Transform=RigVMVariableNode() {variable=knee_slide_jacket_c_rear}.Value
    ItemToModify.1.Transform=RigVMVariableNode() {variable=knee_slide_jacket_l_front}.Value
    ItemToModify.2.Transform=RigVMVariableNode() {variable=knee_slide_jacket_l_rear}.Value
    ItemToModify.3.Transform=RigVMVariableNode() {variable=knee_slide_jacket_r_front}.Value
    ItemToModify.4.Transform=RigVMVariableNode() {variable=knee_slide_jacket_r_rear}.Value
   )
   n10 = &RigVMFunction_ControlFlowBranch(ExecuteContext=@SequenceAggregate.D, Condition=RigVMVariableNode() {variable=HoodDown}.Value)
   HoodieUp = RigVMFunctionReferenceNode(ExecuteContext=@n10.False)
   HoodieDown = RigVMFunctionReferenceNode(ExecuteContext=@n10.True)
   n11 = &RigUnit_SphericalPoseReader(
    ActiveRegionScaleFactors=(PositiveWidth=1,NegativeWidth=3,PositiveHeight=1,NegativeHeight=1)
    ActiveRegionSize=0.05
    Debug=(bDrawDebug=true,DebugScale=25,DebugSegments=20,DebugThickness=0.25)
    DriverAxis=(X=0,Y=1,Z=0)
    DriverItem=(Type=Bone,Name="spine_05")
    ExecutePin=()
    FalloffRegionScaleFactors=(PositiveWidth=1,NegativeWidth=1,PositiveHeight=1,NegativeHeight=1)
    FalloffSize=0.33
    OptionalParentItem=(Type=Bone,Name="pelvis")
    RotationOffset=(X=40,Y=0,Z=0)
   )
   > &RigUnit_ModifyTransforms(
    ItemToModify="((Item=(Type=Bone,Name=\"dyn_sackbag\"),Transform=(Rotation=(X=0,Y=0,Z=0,W=1),Translation=(X=-3,Y=1,Z=2),Scale3D=(X=1,Y=1,Z=1))),(Item=(Type=Bone,Name=\"dyn_sackbelt_l_pec\"),Transform=(Rotation=(X=0,Y=0,Z=0,W=1),Translation=(X=0.5,Y=0,Z=1),Scale3D=(X=1,Y=1,Z= ...(truncated)"
    Weight=/Script/RigVM.RigVMFunction_AnimEasing(SourceMaximum=1, TargetMaximum=1, Type=CubicEaseInOut, Value=@n11.OutputParam) {is=struct RigVMUnitNode}.Result
    ItemToModify.0.Transform.Rotation=&RigVMFunction_MathQuaternionFromEuler(Euler=(X=-5,Y=-20,Z=0)).Result
    ItemToModify.1.Transform.Rotation=&RigVMFunction_MathQuaternionFromEuler(Euler=(X=5,Y=0,Z=5)).Result
   )
   n12 = &RigUnit_SphericalPoseReader(
    ActiveRegionScaleFactors=(PositiveWidth=1,NegativeWidth=1,PositiveHeight=1,NegativeHeight=1)
    ActiveRegionSize=0.1
    Debug=(bDrawDebug=true,DebugScale=25,DebugSegments=20,DebugThickness=0.25)
    DriverAxis=(X=1,Y=0,Z=0)
    DriverItem=(Type=Bone,Name="clavicle_l")
    ExecutePin=()
    FalloffRegionScaleFactors=(PositiveWidth=1,NegativeWidth=1,PositiveHeight=1,NegativeHeight=1)
    FalloffSize=0.5
    OptionalParentItem=(Type=Bone,Name="spine_04")
    RotationOffset=(X=-90,Y=0,Z=0)
   )
   > &RigUnit_ModifyTransforms(
    ItemToModify=((Item=(Type=Bone,Name="dyn_sackbelt_l_scap"),Transform=(Rotation=(X=0,Y=0,Z=0,W=1),Translation=(X=0,Y=0,Z=0),Scale3D=(X=1,Y=1,Z=1))))
    ItemToModify.0.Transform.Rotation=&RigVMFunction_MathQuaternionFromEuler(Euler=(X=0,Y=20,Z=-25)).Result
    Weight=@n12.OutputParam
   )
   &RigVMFunction_MathQuaternionFromEuler(Euler=(X=-90,Y=0,Z=-55), RotationOrder=XYZ)
   &RigVMFunction_MathQuaternionFromEuler(Euler=(X=35,Y=0,Z=0))
   RigUnit_GetTransform = &RigUnit_GetTransform(Item=(Type=Control,Name="thigh_l_ctrl"))
   RigUnit_SetTransform = &RigUnit_SetTransform(ExecutePin=(), Item=(Type=Bone,Name="thigh_l"), Weight=1, bInitial=False, Value=@RigUnit_GetTransform.Transform)
   &RigUnit_SetTransform(
    ExecutePin=()
    Item=(Type=Control,Name="spine_03_ctrl")
    Weight=1
    bInitial=False
    Value=&RigUnit_GetTransform(Item=(Type=Bone,Name="spine_03")).Transform
   )
   RigVMVariableNode() {variable=sackbag_thigh_r_fwd}
   RigVMVariableNode() {variable=jacket_front_leg_fwd}
   RigVMVariableNode() {variable=sackbelt_hip_thigh_r_fwd}
   RigVMVariableNode() {variable=jacket_side_leg_fwd}
   RigVMVariableNode() {variable=sackbelt_lowerback_thigh_r_fwd}
   RigVMVariableNode() {variable=jacket_side_leg_fwd}
  ]}
 SequenceAggregate_1_1_ContainedGraph {nodes [
   n1 = RigVMFunctionEntryNode()
   SequenceAggregate_1 = &RigVMFunction_Sequence(ExecuteContext=@n1.ExecuteContext)
   SequenceAggregate_1_1 = &RigVMFunction_Sequence(ExecuteContext=@SequenceAggregate_1.B)
   RigVMFunctionReturnNode(A=@SequenceAggregate_1.A, B=@SequenceAggregate_1_1.A, C=@SequenceAggregate_1_1.B)
  ]}
 SequenceAggregate_1_ContainedGraph {nodes [
   n1 = RigVMFunctionEntryNode()
   SequenceAggregate = &RigVMFunction_Sequence(ExecuteContext=@n1.ExecuteContext)
   SequenceAggregate_1 = &RigVMFunction_Sequence(ExecuteContext=@SequenceAggregate.B)
   SequenceAggregate_2 = &RigVMFunction_Sequence(ExecuteContext=@SequenceAggregate_1.B)
   RigVMFunctionReturnNode(A=@SequenceAggregate.A, B=@SequenceAggregate_1.A, C=@SequenceAggregate_2.A, D=@SequenceAggregate_2.B)
  ]}
 SequenceAggregate_1_ContainedGraph {nodes [
   n1 = RigVMFunctionEntryNode()
   SequenceAggregate = &RigVMFunction_Sequence(ExecuteContext=@n1.ExecuteContext)
   SequenceAggregate_1 = &RigVMFunction_Sequence(ExecuteContext=@SequenceAggregate.B)
   SequenceAggregate_2 = &RigVMFunction_Sequence(ExecuteContext=@SequenceAggregate_1.B)
   RigVMFunctionReturnNode(A=@SequenceAggregate.A, B=@SequenceAggregate_1.A, C=@SequenceAggregate_2.A, D=@SequenceAggregate_2.B)
  ]}
 SequenceAggregate_1_ContainedGraph {nodes [
   n1 = RigVMFunctionEntryNode()
   SequenceAggregate = &RigVMFunction_Sequence(ExecuteContext=@n1.ExecuteContext)
   SequenceAggregate_1 = &RigVMFunction_Sequence(ExecuteContext=@SequenceAggregate.B)
   SequenceAggregate_2 = &RigVMFunction_Sequence(ExecuteContext=@SequenceAggregate_1.B)
   SequenceAggregate_3 = &RigVMFunction_Sequence(ExecuteContext=@SequenceAggregate_2.B)
   RigVMFunctionReturnNode(A=@SequenceAggregate.A, B=@SequenceAggregate_1.A, C=@SequenceAggregate_2.A, D=@SequenceAggregate_3.A, E=@SequenceAggregate_3.B)
  ]}
]
hierarchy [
 root {loc=<0 0 0>, rot=<0 0 0 1>}
 pelvis {parent=root, loc=<0 -0.7071 93.1191>, rot=<0.0127 0.707 -0.0127 -0.707>}
 spine_01 {parent=pelvis, loc=<6.7494 -0.6388 0>, rot=<0 0 0.034 -0.9994>}
 spine_02 {parent=spine_01, loc=<7.2518 -0.035 0>, rot=<0 0 0.0238 -0.9997>}
 spine_03 {parent=spine_02, loc=<6.0104 0 0>, rot=<0 0 0.0149 -0.9999>}
 spine_04 {parent=spine_03, loc=<8.1024 0 0>, rot=<0 0 -0.0132 -0.9999>}
 spine_05 {parent=spine_04, loc=<10.0146 0 0>, rot=<0 0 -0.0015 -1>}
 attach_cape {parent=spine_05, loc=<9.017 -11.714 0>, rot=<0.0616 -0.7044 -0.0616 -0.7044>}
 neck_01 {parent=spine_05, loc=<11.3138 0.1529 0>, rot=<0 0 -0.2003 -0.9797>}
 neck_02 {parent=neck_01, loc=<4.5627 0 0>, rot=<0 0 0.0504 -0.9987>}
 head {parent=neck_02, loc=<5.3914 0 0>, rot=<0 0 0.115 -0.9934>}
 dyn_hoodie_back_root {parent=head, loc=<10.834 -8.347 0>, rot=<0 0 0.7772 -0.6293>}
 dyn_hoodie_back {parent=dyn_hoodie_back_root, loc=<0 0 0>, rot=<0 0 0 1>}
 dyn_earring_l {parent=head, loc=<1.1148 0.5334 -6.6437>, rot=<0.1002 0.0111 -0.0011 -0.9949>}
 dyn_hair_01 {parent=head, loc=<7.1752 -6.6048 5.0312>, rot=<0.368 -0.2278 0.8962 -0.0976>}
 dyn_hair_02 {parent=head, loc=<12.5595 -5.4982 4.4661>, rot=<0.5535 -0.348 0.7041 -0.2771>}
 dyn_hair_03 {parent=head, loc=<9.7248 -7.2815 -1.2571>, rot=<0.383 -0.3926 0.8148 -0.1877>}
 dyn_hair_04 {parent=head, loc=<7.6017 -6.7652 -3.1959>, rot=<0.3072 -0.316 0.8905 -0.1126>}
 dyn_hair_05 {parent=head, loc=<15.6275 -1.6879 -2.2429>, rot=<0.2121 -0.6037 0.1659 -0.7504>}
 dyn_hair_06 {parent=head, loc=<10.6581 -5.9022 2.8613>, rot=<0.4734 -0.4689 0.6659 -0.3354>}
 dyn_hair_10 {parent=head, loc=<14.473 1.2559 -5.4805>, rot=<0.1238 -0.7052 0.1213 -0.6875>}
 dyn_hair_11 {parent=head, loc=<9.2305 3.689 -9.3605>, rot=<-0.0381 -0.7511 -0.0498 -0.6572>}
 dyn_hair_12 {parent=head, loc=<7.9537 2.8672 -8.8445>, rot=<0.1515 -0.83 0.2607 -0.4693>}
 dyn_hair_13 {parent=head, loc=<6.1558 4.1592 -8.5133>, rot=<-0.0872 -0.8975 -0.3738 -0.2171>}
 dyn_hair_14 {parent=head, loc=<4.5453 -2.9715 -7.4818>, rot=<0.1326 -0.8145 0.5276 -0.2015>}
 dyn_hair_15 {parent=head, loc=<9.49 0.4405 -8.578>, rot=<0.1863 -0.8125 0.343 -0.433>}
 dyn_hair_16 {parent=head, loc=<12.914 -1.5248 -5.886>, rot=<0.1402 -0.7772 0.1795 -0.5865>}
 dyn_hair_17 {parent=head, loc=<13.2309 -2.8094 -4.1328>, rot=<0.2315 -0.7396 0.3037 -0.5542>}
 dyn_hair_18 {parent=head, loc=<12.7028 -4.0281 -2.7865>, rot=<0.1715 -0.8273 0.3299 -0.4211>}
 dyn_hair_19 {parent=head, loc=<12.231 -4.8781 -0.1669>, rot=<0.5564 -0.4162 0.6087 -0.3829>}
 dyn_hair_20 {parent=head, loc=<13.8754 -2.0617 3.2424>, rot=<0.7023 -0.2407 0.607 -0.2834>}
 dyn_hair_21 {parent=head, loc=<15.6275 -0.0219 1.4364>, rot=<0.7213 -0.2432 0.5674 -0.3141>}
 dyn_hair_24 {parent=head, loc=<8.9827 10.2564 -0.657>, rot=<0.1028 0.2474 0.9632 0.0227>}
 dyn_hair_25 {parent=head, loc=<8.8807 10.1856 -2.4838>, rot=<0.3174 0.6261 0.6414 0.3097>}
 dyn_hair_26 {parent=head, loc=<8.0921 8.403 -4.9273>, rot=<0.056 -0.9166 0.3742 -0.1292>}
 dyn_hair_27 {parent=head, loc=<7.6688 10.0526 4.2664>, rot=<0.1684 0.1368 0.976 0.0195>}
 dyn_hair_28 {parent=head, loc=<12.2766 9.7616 3.6532>, rot=<0.5166 0.2878 0.7848 0.1854>}
 dyn_hair_29 {parent=head, loc=<9.9948 3.2813 8.3698>, rot=<0.5013 0.0935 0.8588 0.0499>}
 dyn_hair_30 {parent=head, loc=<11.3851 0.3741 7.5999>, rot=<0.0029 -0.7071 -0.0029 -0.7071>}
 dyn_hair_31 {parent=head, loc=<14.1362 1.3807 5.2978>, rot=<0.7768 -0.0105 0.6293 -0.0195>}
 dyn_hair_32 {parent=head, loc=<10.5112 9.0139 6.0206>, rot=<0.3376 0.2246 0.9107 0.0792>}
 dyn_hair_33 {parent=head, loc=<11.5673 8.2779 7.2376>, rot=<0.6896 0.1523 0.6927 0.1462>}
 dyn_hair_34 {parent=head, loc=<15.9798 2.6411 4.4975>, rot=<0.8398 -0.0277 0.5399 -0.0506>}
 dyn_earring_r {parent=head, loc=<1.1144 0.5334 6.6437>, rot=<0.0111 -0.1002 0.9949 -0.0011>}
 dyn_hair_08 {parent=head, loc=<6.5663 -4.8442 2.1406>, rot=<0.0041 -0.9979 -0.0003 -0.0645>}
 dyn_hair_07 {parent=head, loc=<6.5663 -4.8442 -4.8014>, rot=<0.1256 -0.9919 0.0022 -0.0182>}
 dyn_hair_09 {parent=head, loc=<6.6984 3.163 -7.0121>, rot=<-0.0159 -0.9995 0.0008 -0.028>}
 dyn_hair_22 {parent=head, loc=<8.1217 2.8686 7.3134>, rot=<0.0688 -0.8182 0.5632 0.0928>}
 dyn_hair_23 {parent=head, loc=<7.4974 -2.5785 7.3134>, rot=<0.0378 -0.8403 0.5365 -0.068>}
 dyn_hair_40 {parent=head, loc=<8.9808 6.7658 -6.0842>, rot=<0.0877 -0.8954 0.3911 -0.1936>}
 dyn_hair_41 {parent=head, loc=<12.9937 6.6618 7.1473>, rot=<0.8341 0.0413 0.5472 0.0555>}
 dyn_hair_35 {parent=head, loc=<12.6946 8.04 -4.796>, rot=<-0.15 -0.834 -0.2876 -0.4463>}
 dyn_hair_36 {parent=head, loc=<13.4731 6.758 -5.7876>, rot=<-0.0535 -0.694 -0.0576 -0.7157>}
 dyn_hair_37 {parent=head, loc=<14.6998 5.1405 -4.9415>, rot=<0.1238 -0.5981 0.0891 -0.7868>}
 dyn_hair_38 {parent=head, loc=<14.4662 8.5391 3.5331>, rot=<0.7955 0.1703 0.5239 0.2522>}
 dyn_hair_39 {parent=head, loc=<8.2354 8.195 6.1585>, rot=<0.1761 -0.4326 0.8797 -0.0895>}
 deform_hoodie_up_front_root {parent=neck_01, loc=<5.4884 9.1436 0>, rot=<0 0 0.0504 -0.9987>}
 deform_hoodie_up_front {parent=deform_hoodie_up_front_root, loc=<0 0 0>, rot=<0 0 0 1>}
 deform_hoodie_up_front_edge {parent=deform_hoodie_up_front, loc=<5.5136 0 0>, rot=<0 0 0 1>}
 deform_hoodie_up_back_root {parent=neck_01, loc=<4.0264 -6.19 0>, rot=<0 0 0.9987 0.0504>}
 deform_hoodie_up_back {parent=deform_hoodie_up_back_root, loc=<0 0 0>, rot=<0 0 0 1>}
 attach_backpack {parent=spine_05, loc=<-0.123 -13.473 0>, rot=<-0.0093 -0.707 0.0093 -0.707>}
 clavicle_l {parent=spine_05, loc=<4.3575 -0.1466 -2.3223>, rot=<0.0465 -0.7289 -0.009 -0.6829>}
 upperarm_l {parent=clavicle_l, loc=<15.5111 0 0>, rot=<0.0226 -0.3234 -0.0088 -0.9459>}
 lowerarm_l {parent=upperarm_l, loc=<26.6514 0 0>, rot=<0.0391 -0.0099 -0.2453 -0.9686>}
 hand_l {parent=lowerarm_l, loc=<22.3491 0 0>, rot=<-0.7081 -0.0507 0.0975 -0.6975>}
 thumb_01_l {parent=hand_l, loc=<1.3118 -0.3191 -2.459>, rot=<0.6546 -0.4042 0.0664 -0.6354>}
 thumb_02_l {parent=thumb_01_l, loc=<5.9336 -0.0448 0.0401>, rot=<-0.0037 0.0213 0.298 -0.9543>}
 thumb_03_l {parent=thumb_02_l, loc=<2.2223 0.3682 -0.0907>, rot=<0.0118 -0.013 0.0066 -0.9998>}
 index_metacarpal_l {parent=hand_l, loc=<1.6535 0.0183 -2.96>, rot=<0.0158 -0.0124 -0.0173 -0.9996>}
 index_01_l {parent=index_metacarpal_l, loc=<8.3008 0.0357 -0.0524>, rot=<-0.0044 0.0485 0.0924 -0.9945>}
 index_02_l {parent=index_01_l, loc=<3.3587 -0.3004 0.0452>, rot=<-0.002 0.0027 0.3225 -0.9466>}
 index_03_l {parent=index_02_l, loc=<2.5578 -0.123 -0.0252>, rot=<-0.002 0.0067 0.2558 -0.9667>}
 middle_metacarpal_l {parent=hand_l, loc=<2.1856 -0.1725 -0.5984>, rot=<0.006 0.0168 -0.0329 -0.9993>}
 middle_01_l {parent=middle_metacarpal_l, loc=<7.3275 -0.2447 0.0129>, rot=<-0.0208 0.0042 0.0858 -0.9961>}
 middle_02_l {parent=middle_01_l, loc=<4.0123 0.1401 0.1175>, rot=<-0.0218 0.0258 0.3512 -0.9357>}
 middle_03_l {parent=middle_02_l, loc=<2.6077 0.0534 0.1373>, rot=<0.0082 0.0188 0.1708 -0.9851>}
 ring_metacarpal_l {parent=hand_l, loc=<2.2074 -0.3625 1.3638>, rot=<-0.0284 0.0722 -0.0458 -0.9959>}
 ring_01_l {parent=ring_metacarpal_l, loc=<7.0295 -0.6197 -0.1446>, rot=<-0.0489 -0.0252 0.1892 -0.9804>}
 ring_02_l {parent=ring_01_l, loc=<3.6174 0.4647 -0.1027>, rot=<0 0 0.2806 -0.9598>}
 ring_03_l {parent=ring_02_l, loc=<2.4152 0.114 0.0199>, rot=<0 0 0.1189 -0.9929>}
 pinky_metacarpal_l {parent=hand_l, loc=<1.7616 -0.9753 3.3814>, rot=<-0.0397 0.0814 -0.069 -0.9935>}
 pinky_01_l {parent=pinky_metacarpal_l, loc=<7.1709 -0.4831 -0.1094>, rot=<-0.0654 -0.0035 0.2242 -0.9723>}
 pinky_02_l {parent=pinky_01_l, loc=<2.8274 0.0893 -0.1043>, rot=<0.0037 0.0011 0.2854 -0.9584>}
 pinky_03_l {parent=pinky_02_l, loc=<2.2046 0.0307 -0.0078>, rot=<-0.0002 0.0098 0.1445 -0.9895>}
 deform_wrist_dn_base_l {parent=hand_l, loc=<0 -2.2342 0>, rot=<0 0 0 -1>}
 deform_wrist_dn_l {parent=deform_wrist_dn_base_l, loc=<0 0 0>, rot=<0 0 0 1>}
 deform_wrist_up_base_l {parent=hand_l, loc=<0 2.2682 0>, rot=<0 0 0 -1>}
 deform_wrist_up_l {parent=deform_wrist_up_base_l, loc=<0 0 0>, rot=<0 0 0 1>}
 hand_attach_l {parent=hand_l, loc=<8.5 -3 0>, rot=<0.7071 0.7071 0 0>}
 weapon_l {parent=hand_l, loc=<-0.1974 -3.4111 0.3811>, rot=<0 0 0 1>}
 lowerarm_twist_02_l {parent=lowerarm_l, loc=<9.6199 0 0>, rot=<0 0 0 -1>}
 lowerarm_twist_01_l {parent=lowerarm_l, loc=<17.8549 0 0>, rot=<0 0 0 -1>}
 deform_elbow_in_base_l {parent=lowerarm_l, loc=<0.2112 5.259 0.8274>, rot=<0 0 0 1>}
 deform_elbow_in_l {parent=deform_elbow_in_base_l, loc=<0 0 0>, rot=<0 0 0 1>}
 deform_elbow_out_base_l {parent=lowerarm_l, loc=<0 -4 0>, rot=<0 0 0 1>}
 deform_elbow_out_l {parent=deform_elbow_out_base_l, loc=<0 0 0>, rot=<0 0 0 1>}
 upperarm_twist_01_l {parent=upperarm_l, loc=<0.0041 0 0>, rot=<0 0 0 1>}
 deform_upperarm_bk_base_l {parent=upperarm_twist_01_l, loc=<1.0813 -7.2193 -1.9453>, rot=<0 0 0 1>}
 deform_upperarm_bk_l {parent=deform_upperarm_bk_base_l, loc=<0 0 0>, rot=<0 0 0 1>}
 deform_upperarm_fr_base_l {parent=upperarm_twist_01_l, loc=<-0.3691 7.0506 -2.765>, rot=<0 0 0 1>}
 deform_upperarm_fr_l {parent=deform_upperarm_fr_base_l, loc=<0 0 0>, rot=<0 0 0 1>}
 deform_upperarm_up_base_l {parent=upperarm_twist_01_l, loc=<-3.2216 -1.1391 3.3311>, rot=<0 0 0 1>}
 deform_upperarm_up_l {parent=deform_upperarm_up_base_l, loc=<0 0 0>, rot=<0 0 0 1>}
 upperarm_twist_02_l {parent=upperarm_l, loc=<13.7941 0 0>, rot=<0 0 0 1>}
 dyn_l_arm_strap_root {parent=upperarm_l, loc=<9.0394 -2.1222 5.1057>, rot=<-0.1388 0.3842 0.1186 -0.905>}
 dyn_l_arm_strap_01 {parent=dyn_l_arm_strap_root, loc=<0 0 0>, rot=<0.2181 0.0237 0.0377 -0.9749>}
 dyn_l_arm_strap_02 {parent=dyn_l_arm_strap_01, loc=<4.904 0 0>, rot=<-0.0169 -0.2605 -0.0358 -0.9647>}
 dyn_l_arm_strap_03 {parent=dyn_l_arm_strap_02, loc=<4.4797 0 0>, rot=<0.0122 -0.2259 -0.0524 -0.9727>}
 dyn_sackbelt_l_pec_root {parent=clavicle_l, loc=<8.7143 12.0817 -3.093>, rot=<0.4757 0.2114 0.4409 -0.7312>}
 dyn_sackbelt_l_pec {parent=dyn_sackbelt_l_pec_root, loc=<0 0 0>, rot=<0 0 0 1>}
 dyn_sackbelt_l_clav_root {parent=clavicle_l, loc=<14.3699 -0.8859 8.0303>, rot=<-0.5861 -0.0261 0.8086 -0.0445>}
 dyn_sackbelt_l_clav {parent=dyn_sackbelt_l_clav_root, loc=<0 0 0>, rot=<0 0 0 1>}
 dyn_sackbelt_l_scap_root {parent=clavicle_l, loc=<10.4611 -10.4776 -0.3824>, rot=<-0.5386 0.011 0.8392 -0.0745>}
 dyn_sackbelt_l_scap {parent=dyn_sackbelt_l_scap_root, loc=<0 0 0>, rot=<0 0 0 1>}
 dyn_l_hoodie_clav_root {parent=clavicle_l, loc=<9.0842 -1.2806 9.3962>, rot=<-0.0061 -0.003 0.7158 -0.6983>}
 dyn_l_hoodie_clav {parent=dyn_l_hoodie_clav_root, loc=<0 0 0>, rot=<0 0 0 1>}
 clavicle_r {parent=spine_05, loc=<4.3575 -0.1466 2.3223>, rot=<0.7289 0.0465 -0.6829 0.009>}
 upperarm_r {parent=clavicle_r, loc=<-15.5111 0 0>, rot=<0.0226 -0.3234 -0.0088 -0.9459>}
 lowerarm_r {parent=upperarm_r, loc=<-26.6514 0 0>, rot=<0.0391 -0.0099 -0.2453 -0.9686>}
 hand_r {parent=lowerarm_r, loc=<-22.3491 0 0>, rot=<-0.7081 -0.0507 0.0975 -0.6975>}
 thumb_01_r {parent=hand_r, loc=<-1.3118 0.3191 2.459>, rot=<0.6546 -0.4042 0.0664 -0.6354>}
 thumb_02_r {parent=thumb_01_r, loc=<-5.9336 0.0448 -0.0401>, rot=<-0.0037 0.0213 0.298 -0.9543>}
 thumb_03_r {parent=thumb_02_r, loc=<-2.2223 -0.3682 0.0907>, rot=<0.0118 -0.013 0.0066 -0.9998>}
 index_metacarpal_r {parent=hand_r, loc=<-1.6535 -0.0183 2.96>, rot=<0.0158 -0.0124 -0.0173 -0.9996>}
 index_01_r {parent=index_metacarpal_r, loc=<-8.3008 -0.0357 0.0524>, rot=<-0.0044 0.0485 0.0924 -0.9945>}
 index_02_r {parent=index_01_r, loc=<-3.3587 0.3004 -0.0452>, rot=<-0.002 0.0027 0.3225 -0.9466>}
 index_03_r {parent=index_02_r, loc=<-2.5578 0.123 0.0252>, rot=<-0.002 0.0067 0.2558 -0.9667>}
 middle_metacarpal_r {parent=hand_r, loc=<-2.1856 0.1725 0.5984>, rot=<0.006 0.0168 -0.0329 -0.9993>}
 middle_01_r {parent=middle_metacarpal_r, loc=<-7.3275 0.2447 -0.0129>, rot=<-0.0208 0.0042 0.0858 -0.9961>}
 middle_02_r {parent=middle_01_r, loc=<-4.0123 -0.1401 -0.1175>, rot=<-0.0218 0.0258 0.3512 -0.9357>}
 middle_03_r {parent=middle_02_r, loc=<-2.6077 -0.0534 -0.1373>, rot=<0.0082 0.0188 0.1708 -0.9851>}
 ring_metacarpal_r {parent=hand_r, loc=<-2.2074 0.3625 -1.3638>, rot=<-0.0284 0.0722 -0.0458 -0.9959>}
 ring_01_r {parent=ring_metacarpal_r, loc=<-7.0295 0.6197 0.1446>, rot=<-0.0489 -0.0252 0.1892 -0.9804>}
 ring_02_r {parent=ring_01_r, loc=<-3.6174 -0.4647 0.1027>, rot=<0 0 0.2806 -0.9598>}
 ring_03_r {parent=ring_02_r, loc=<-2.4152 -0.114 -0.0199>, rot=<0 0 0.1189 -0.9929>}
 pinky_metacarpal_r {parent=hand_r, loc=<-1.7616 0.9753 -3.3814>, rot=<-0.0397 0.0814 -0.069 -0.9935>}
 pinky_01_r {parent=pinky_metacarpal_r, loc=<-7.1709 0.4831 0.1094>, rot=<-0.0654 -0.0035 0.2242 -0.9723>}
 pinky_02_r {parent=pinky_01_r, loc=<-2.8274 -0.0893 0.1043>, rot=<0.0037 0.0011 0.2854 -0.9584>}
 pinky_03_r {parent=pinky_02_r, loc=<-2.2046 -0.0307 0.0078>, rot=<-0.0002 0.0098 0.1445 -0.9895>}
 deform_wrist_dn_base_r {parent=hand_r, loc=<-0.0004 2.2345 -0.0001>, rot=<0 0 0 -1>}
 deform_wrist_dn_r {parent=deform_wrist_dn_base_r, loc=<0 0 0>, rot=<0 0 0 1>}
 deform_wrist_up_base_r {parent=hand_r, loc=<0.0003 -2.2684 0.0001>, rot=<0 0 0 -1>}
 deform_wrist_up_r {parent=deform_wrist_up_base_r, loc=<0 0 0>, rot=<0 0 0 1>}
 weapon_r {parent=hand_r, loc=<0.1974 3.4111 -0.3811>, rot=<0 0 0 1>}
 hand_attach_r {parent=hand_r, loc=<-8.5001 2.9999 0>, rot=<0 0 -0.7071 -0.7071>}
 lowerarm_twist_02_r {parent=lowerarm_r, loc=<-9.6199 0 0>, rot=<0 0 0 -1>}
 lowerarm_twist_01_r {parent=lowerarm_r, loc=<-17.8549 0 0>, rot=<0 0 0 -1>}
 deform_elbow_in_base_r {parent=lowerarm_r, loc=<-0.2111 -5.2591 -0.8275>, rot=<0 0 0 -1>}
 deform_elbow_in_r {parent=deform_elbow_in_base_r, loc=<-0.0001 0 0.0002>, rot=<0 0 0 1>}
 deform_elbow_out_base_r {parent=lowerarm_r, loc=<0 4 -0.0001>, rot=<0 0 0 -1>}
 deform_elbow_out_r {parent=deform_elbow_out_base_r, loc=<0 0 0>, rot=<0 0 0 1>}
 upperarm_twist_01_r {parent=upperarm_r, loc=<-0.0041 0 0>, rot=<0 0 0 -1>}
 deform_upperarm_bk_base_r {parent=upperarm_twist_01_r, loc=<-1.0814 7.2192 1.9455>, rot=<0 0 0 1>}
 deform_upperarm_bk_r {parent=deform_upperarm_bk_base_r, loc=<0 0 0>, rot=<0 0 0 1>}
 deform_upperarm_fr_base_r {parent=upperarm_twist_01_r, loc=<0.3689 -7.0506 2.7652>, rot=<0 0 0 1>}
 deform_upperarm_fr_r {parent=deform_upperarm_fr_base_r, loc=<0 0 0>, rot=<0 0 0 1>}
 deform_upperarm_up_base_r {parent=upperarm_twist_01_r, loc=<3.2214 1.1391 -3.3308>, rot=<0 0 0 1>}
 deform_upperarm_up_r {parent=deform_upperarm_up_base_r, loc=<0 0 0>, rot=<0 0 0 -1>}
 upperarm_twist_02_r {parent=upperarm_r, loc=<-13.7941 0 0>, rot=<0 0 0 -1>}
 dyn_r_arm_strap_root {parent=upperarm_r, loc=<-9.0393 2.1222 -5.1058>, rot=<-0.1388 0.3842 0.1186 -0.905>}
 dyn_r_arm_strap_01 {parent=dyn_r_arm_strap_root, loc=<0 0 0>, rot=<0.2181 0.0237 0.0377 -0.9749>}
 dyn_r_arm_strap_02 {parent=dyn_r_arm_strap_01, loc=<-4.9039 -0.0001 0.0005>, rot=<-0.0169 -0.2605 -0.0358 -0.9647>}
 dyn_r_arm_strap_03 {parent=dyn_r_arm_strap_02, loc=<-4.4797 0 0>, rot=<0.0122 -0.2259 -0.0524 -0.9727>}
 dyn_r_hoodie_clav_root {parent=clavicle_r, loc=<-9.0842 1.2806 -9.3959>, rot=<-0.0061 -0.003 0.7158 -0.6983>}
 dyn_r_hoodie_clav {parent=dyn_r_hoodie_clav_root, loc=<0 0 0>, rot=<0 0 0 1>}
 deform_hoodie_up_sternum {parent=spine_05, loc=<4.8203 9.2881 0>, rot=<0 0 -0.1506 -0.9886>}
 dyn_hoodie_down_root {parent=spine_04, loc=<24.4409 -10.8337 0>, rot=<0 0 0 1>}
 dyn_hoodie_down_01 {parent=dyn_hoodie_down_root, loc=<0 0 0>, rot=<-0.0018 -0.0102 -0.0988 -0.9951>}
 dyn_hoodie_down_02 {parent=dyn_hoodie_down_01, loc=<-5.4008 0 0>, rot=<0 0 0 1>}
 dyn_sackbelt_l_lowerback_root {parent=spine_04, loc=<1.0187 -12.3739 -3.7748>, rot=<0.351 0.0055 0.9328 0.0811>}
 dyn_sackbelt_l_lowerback {parent=dyn_sackbelt_l_lowerback_root, loc=<0 0 0>, rot=<0 0 0 1>}
 deform_pec_aim_base_l {parent=spine_04, loc=<7.8732 12.6388 -2.4123>, rot=<0.0102 -0.4602 0.3979 -0.7936>}
 deform_pec_aim_l {parent=deform_pec_aim_base_l, loc=<0 0 0>, rot=<0 0 0 1>}
 deform_pec_base_l {parent=deform_pec_aim_l, loc=<10.766 0 0>, rot=<0 0 0 1>}
 deform_pec_l {parent=deform_pec_base_l, loc=<0.055 0 0>, rot=<0 0 0 1>}
 deform_pec_aim_base_r {parent=spine_04, loc=<7.8727 12.6388 2.4123>, rot=<-0.4602 -0.0102 0.7936 0.3979>}
 deform_pec_aim_r {parent=deform_pec_aim_base_r, loc=<0 0 0>, rot=<0 0 0 1>}
 deform_pec_base_r {parent=deform_pec_aim_r, loc=<-10.7662 -0.0006 -0.0007>, rot=<0 0 0 1>}
 deform_pec_r {parent=deform_pec_base_r, loc=<0 0 0>, rot=<0 0 0 1>}
 jacket_l_front_01_root {parent=spine_03, loc=<-2.4603 7.7226 -8.7492>, rot=<0.0861 -0.8476 -0.5236 -0.0074>}
 jacket_l_front_01 {parent=jacket_l_front_01_root, loc=<0 0 0>, rot=<0 0 0 1>}
 jacket_l_front_02 {parent=jacket_l_front_01, loc=<9 0 0>, rot=<0 0.0267 0 -0.9996>}
 jacket_l_front_03 {parent=jacket_l_front_02, loc=<9 0 0>, rot=<0 0.0283 0 -0.9996>}
 jacket_l_front_04 {parent=jacket_l_front_03, loc=<9 0 0>, rot=<0 0 0 1>}
 jacket_l_front_05 {parent=jacket_l_front_04, loc=<9 0 0>, rot=<0 0 0 1>}
 jacket_l_side_01_root {parent=spine_03, loc=<-1.6131 -1.646 -12.4667>, rot=<0.0487 -0.9943 0.0822 -0.0469>}
 jacket_l_side_01 {parent=jacket_l_side_01_root, loc=<0 0 0>, rot=<0 0 0 1>}
 jacket_l_side_02 {parent=jacket_l_side_01, loc=<9 0 0>, rot=<0 0.0632 0 -0.998>}
 jacket_l_side_03 {parent=jacket_l_side_02, loc=<9 0 0>, rot=<0 0.0283 0 -0.9996>}
 jacket_l_side_04 {parent=jacket_l_side_03, loc=<9 0 0>, rot=<0 0.0185 0 -0.9998>}
 jacket_l_side_05 {parent=jacket_l_side_04, loc=<9 0 0>, rot=<0 0 0 1>}
 jacket_l_side_06 {parent=jacket_l_side_05, loc=<7.7663 -5.8583 0>, rot=<0 0 0 1>}
 jacket_l_side_07 {parent=jacket_l_side_05, loc=<13.0397 2.1527 0.4053>, rot=<0 0 0 1>}
 jacket_l_rear_01_root {parent=spine_03, loc=<-1.2465 -7.6342 -8.324>, rot=<0.0549 -0.8097 0.5814 -0.0573>}
 jacket_l_rear_01 {parent=jacket_l_rear_01_root, loc=<0 0 0>, rot=<0 0 0 1>}
 jacket_l_rear_02 {parent=jacket_l_rear_01, loc=<9 0 0>, rot=<0 0.0675 0 -0.9977>}
 jacket_l_rear_03 {parent=jacket_l_rear_02, loc=<9 0 0>, rot=<0 0.0283 0 -0.9996>}
 jacket_l_rear_04 {parent=jacket_l_rear_03, loc=<9 0 0>, rot=<0 0 0 1>}
 jacket_l_rear_05 {parent=jacket_l_rear_04, loc=<9 0 0>, rot=<0 0 0 1>}
 jacket_c_rear_01_root {parent=spine_03, loc=<-1.2448 -9.4669 -1.2379>, rot=<0.0499 -0.7231 0.6876 -0.0434>}
 jacket_c_rear_01 {parent=jacket_c_rear_01_root, loc=<0 0 0>, rot=<0 0 0 1>}
 jacket_c_rear_02 {parent=jacket_c_rear_01, loc=<9 0 0>, rot=<0 0.0844 0 -0.9964>}
 jacket_c_rear_03 {parent=jacket_c_rear_02, loc=<9 0 0>, rot=<0 0.0283 0 -0.9996>}
 jacket_c_rear_04 {parent=jacket_c_rear_03, loc=<9 0 0>, rot=<0 0 0 1>}
 jacket_c_rear_05 {parent=jacket_c_rear_04, loc=<6.4414 0 0>, rot=<0 0 0 1>}
 jacket_c_rear_06 {parent=jacket_c_rear_05, loc=<11.3026 3.0147 0>, rot=<0 0 0 1>}
 jacket_r_front_01_root {parent=spine_03, loc=<-2.4601 7.7226 8.7492>, rot=<-0.8476 -0.0861 0.0074 -0.5236>}
 jacket_r_front_01 {parent=jacket_r_front_01_root, loc=<0 0 0>, rot=<0 0 0 1>}
 jacket_r_front_02 {parent=jacket_r_front_01, loc=<-9.0006 -0.0001 0>, rot=<0 0.0267 0 -0.9996>}
 jacket_r_front_03 {parent=jacket_r_front_02, loc=<-8.9996 0 0>, rot=<0 0.0283 0 -0.9996>}
 jacket_r_front_04 {parent=jacket_r_front_03, loc=<-9 0.0001 0>, rot=<0 0 0 -1>}
 jacket_r_front_05 {parent=jacket_r_front_04, loc=<-9 0 0>, rot=<0 0 0 -1>}
 jacket_r_side_01_root {parent=spine_03, loc=<-1.613 -1.646 12.4667>, rot=<0.9943 0.0487 -0.0469 -0.0822>}
 jacket_r_side_01 {parent=jacket_r_side_01_root, loc=<0 0 0>, rot=<-0.0189 0 0 -0.9998>}
 jacket_r_side_02 {parent=jacket_r_side_01, loc=<-8.9999 0 -0.0001>, rot=<0 0.0632 0 -0.998>}
 jacket_r_side_03 {parent=jacket_r_side_02, loc=<-9.0002 0 0.0001>, rot=<0 0.0283 0 -0.9996>}
 jacket_r_side_04 {parent=jacket_r_side_03, loc=<-9 0 -0.0001>, rot=<0 0.0185 0 -0.9998>}
 jacket_r_side_05 {parent=jacket_r_side_04, loc=<-9 0 0>, rot=<0 0 0 1>}
 jacket_r_side_06 {parent=jacket_r_side_05, loc=<-7.7662 5.8583 0>, rot=<0 0 0 1>}
 jacket_r_side_07 {parent=jacket_r_side_05, loc=<-11.5901 -5.53 1.446>, rot=<0.2766 -0.0149 -0.0517 -0.9595>}
 jacket_r_rear_01_root {parent=spine_03, loc=<-1.2462 -7.6342 8.324>, rot=<0.8097 0.0549 -0.0573 -0.5814>}
 jacket_r_rear_01 {parent=jacket_r_rear_01_root, loc=<0 0 0>, rot=<0 0 0 1>}
 jacket_r_rear_02 {parent=jacket_r_rear_01, loc=<-9.0006 0 0>, rot=<0 0.0675 0 -0.9977>}
 jacket_r_rear_03 {parent=jacket_r_rear_02, loc=<-8.9997 0 -0.0001>, rot=<0 0.0283 0 -0.9996>}
 jacket_r_rear_04 {parent=jacket_r_rear_03, loc=<-9 0 0>, rot=<0 0 0 -1>}
 jacket_r_rear_05 {parent=jacket_r_rear_04, loc=<-9 -0.0001 0>, rot=<0 0 0 -1>}
 jacket_r_rear_06 {parent=jacket_r_rear_05, loc=<-8.9739 7.3011 0>, rot=<0 0 0 1>}
 jacket_r_rear_07 {parent=jacket_r_rear_05, loc=<-9.0011 -3.0237 0>, rot=<0 0 0 1>}
 dyn_sackbag_root {parent=spine_03, loc=<2.2256 16.81 0.8621>, rot=<-0.0704 -0.4227 -0.0658 -0.9011>}
 dyn_sackbag {parent=dyn_sackbag_root, loc=<0 0 0>, rot=<0 0 0 1>}
 dyn_sackbelt_r_hip_root {parent=spine_03, loc=<-10.1148 -3.1764 15.0762>, rot=<-0.0786 -0.451 -0.205 -0.8651>}
 dyn_sackbelt_r_hip {parent=dyn_sackbelt_r_hip_root, loc=<0 0 0>, rot=<0 0 0 1>}
 dyn_sackbelt_r_lowerback_root {parent=spine_03, loc=<-2.4597 -10.174 6.6873>, rot=<0.3993 0.1213 0.9071 0.0544>}
 dyn_sackbelt_r_lowerback {parent=dyn_sackbelt_r_lowerback_root, loc=<0 0 0>, rot=<0 0 0 1>}
 thigh_l {parent=pelvis, loc=<-2.5693 -0.0109 -7.7501>, rot=<0.0012 0.0572 0.0155 -0.9982>}
 calf_l {parent=thigh_l, loc=<-39.0484 0 0>, rot=<0.0006 -0.0161 -0.0391 -0.9991>}
 foot_l {parent=calf_l, loc=<-42.561 0.1077 0.4292>, rot=<0.0418 -0.0455 0.0343 -0.9975>}
 ball_l {parent=foot_l, loc=<-6.1599 13.8264 0.6806>, rot=<0.0001 -0.0001 -0.7063 -0.7079>}
 calf_twist_02_l {parent=calf_l, loc=<-18.8064 -0.0967 0.0397>, rot=<0.0237 0.0095 -0.0035 -0.9997>}
 calf_twist_01_l {parent=calf_l, loc=<-35.813 -0.0967 0.0397>, rot=<0.0237 0.0095 -0.0035 -0.9997>}
 dyn_l_leg_strap_root {parent=calf_l, loc=<-7.1732 -3.3212 -6.2705>, rot=<0.0359 -0.8915 0.1804 -0.414>}
 dyn_l_leg_strap_01 {parent=dyn_l_leg_strap_root, loc=<0 0 0>, rot=<0 0 0 1>}
 dyn_l_leg_strap_02 {parent=dyn_l_leg_strap_01, loc=<3.6433 0 0>, rot=<0.0442 -0.3597 -0.1212 -0.9241>}
 dyn_l_leg_strap_03 {parent=dyn_l_leg_strap_02, loc=<3.5931 0 0>, rot=<0 0 0 1>}
 deform_knee_in_base_l {parent=calf_l, loc=<1.298 -5.2957 -0.0496>, rot=<0 0 0 -1>}
 deform_knee_in_l {parent=deform_knee_in_base_l, loc=<0 0 0>, rot=<0 0 0 1>}
 deform_knee_out_base_l {parent=calf_l, loc=<0 4 0>, rot=<0 0 0 -1>}
 deform_knee_out_l {parent=deform_knee_out_base_l, loc=<0 0 0>, rot=<0 0 0 1>}
 thigh_twist_01_l {parent=thigh_l, loc=<-21.6303 0 0>, rot=<0.0013 0 0 -1>}
 deform_groin_base_l {parent=thigh_l, loc=<-2.1908 8.1142 -1.7442>, rot=<0 0 0 -1>}
 deform_groin_l {parent=deform_groin_base_l, loc=<0 0 0>, rot=<0 0 0 1>}
 dyn_thigh_l_zipper_root {parent=thigh_l, loc=<-32.3952 2.6273 6.302>, rot=<0.385 -0.0063 -0.0243 -0.9226>}
 dyn_thigh_l_zipper {parent=dyn_thigh_l_zipper_root, loc=<0 0 0>, rot=<0 0 0 1>}
 NewControl {controlType=EULER_TRANSFORM, parent=thigh_l, type=Control, loc=<0 0 0>, rot=<0 0 0 1>}
 thigh_r {parent=pelvis, loc=<-2.5693 -0.0109 7.7501>, rot=<0.0572 -0.0012 0.9982 0.0155>}
 calf_r {parent=thigh_r, loc=<39.0484 0 0>, rot=<0.0006 -0.0161 -0.0391 -0.9991>}
 foot_r {parent=calf_r, loc=<42.561 -0.1077 -0.4292>, rot=<0.0418 -0.0455 0.0343 -0.9975>}
 ball_r {parent=foot_r, loc=<6.1599 -13.8264 -0.6806>, rot=<0.0001 -0.0001 -0.7063 -0.7079>}
 calf_twist_02_r {parent=calf_r, loc=<18.8064 0.0967 -0.0397>, rot=<0.0237 0.0095 -0.0035 -0.9997>}
 calf_twist_01_r {parent=calf_r, loc=<35.813 0.0967 -0.0397>, rot=<0.0237 0.0095 -0.0035 -0.9997>}
 deform_knee_in_base_r {parent=calf_r, loc=<-1.298 5.2957 0.0496>, rot=<0 0 0 -1>}
 deform_knee_in_r {parent=deform_knee_in_base_r, loc=<0 0 0>, rot=<0 0 0 1>}
 deform_knee_out_base_r {parent=calf_r, loc=<0 -4 0>, rot=<0 0 0 1>}
 deform_knee_out_r {parent=deform_knee_out_base_r, loc=<0 0 0>, rot=<0 0 0 1>}
 thigh_twist_01_r {parent=thigh_r, loc=<21.6303 0 0>, rot=<0.0013 0 0 -1>}
 deform_groin_base_r {parent=thigh_r, loc=<2.1908 -8.1142 1.7442>, rot=<0 0 0 -1>}
 deform_groin_r {parent=deform_groin_base_r, loc=<0 0 0>, rot=<0 0 0 1>}
 dyn_thigh_r_zipper_root {parent=thigh_r, loc=<32.3951 -2.6273 -6.302>, rot=<0.385 -0.0063 -0.0243 -0.9226>}
 dyn_thigh_r_zipper {parent=dyn_thigh_r_zipper_root, loc=<0 0 0>, rot=<0 0 0 1>}
 deform_glute_base_l {parent=pelvis, loc=<-8.8064 -8.8222 -8.4287>, rot=<0.0275 0.0501 -0.4472 -0.8926>}
 deform_glute_l {parent=deform_glute_base_l, loc=<0 0 0>, rot=<0 0 0 1>}
 deform_glute_base_r {parent=pelvis, loc=<-8.8062 -8.8222 8.4287>, rot=<0.0501 -0.0275 0.8926 -0.4472>}
 deform_glute_r {parent=deform_glute_base_r, loc=<0 0 0>, rot=<0 0 0 1>}
 deform_shirt_tatters_root {parent=pelvis, loc=<-9.6955 8.8735 -0.0361>, rot=<0 0 0 1>}
 deform_shirt_tatters {parent=deform_shirt_tatters_root, loc=<0 0 0>, rot=<0 0 0 1>}
 attach {parent=root, loc=<0 0 0>, rot=<0 0 0 1>}
 ik_hand_root {parent=root, loc=<0 0 0>, rot=<0 0 0 1>}
 ik_hand_gun {parent=ik_hand_root, loc=<-51.9727 4.0883 102.6935>, rot=<0.6084 0.1793 0.4613 -0.6205>}
 ik_hand_r {parent=ik_hand_gun, loc=<0 0 0>, rot=<0 0 0 1>}
 ik_hand_l {parent=ik_hand_gun, loc=<53.0285 82.1781 35.2051>, rot=<-0.5102 -0.7906 -0.3387 0>}
 ik_foot_root {parent=root, loc=<0 0 0>, rot=<0 0 0 1>}
 ik_foot_l {parent=ik_foot_root, loc=<15.2608 -4.4176 9.4432>, rot=<0.0357 0.7045 0.0284 -0.7082>}
 ik_foot_r {parent=ik_foot_root, loc=<-15.2608 -4.4176 9.4432>, rot=<0.7082 0.0284 -0.7045 0.0357>}
 attach_fp {parent=root, loc=<0 0 0>, rot=<0 0 0 1>}
 camera_root {parent=root, loc=<0 0 0>, rot=<0 0 0 1>}
 camera {parent=camera_root, loc=<0 0 0>, rot=<0 0 0 1>}
 warp_point_01 {parent=root, loc=<0 0 0>, rot=<0 0 0 1>}
 warp_point_02 {parent=root, loc=<0 0 0>, rot=<0 0 0 1>}
 warp_point_03 {parent=root, loc=<0 0 0>, rot=<0 0 0 1>}
 thigh_l_ctrl {controlType=EULER_TRANSFORM, type=Control, loc=<0 0 0>, rot=<0 0 0 1>}
 spine_03_ctrl {controlType=EULER_TRANSFORM, type=Control, loc=<0 0 0>, rot=<0 0 0 1>}
 jacket_l_front_01_root_ctrl_2 {controlType=EULER_TRANSFORM, parent=spine_03_ctrl, type=Control, loc=<8.7492 6.0938 111.5238>, rot=<-0.3974 0.5763 0.3414 0.6272>}
 jacket_l_front_01_root_ctrl {controlType=EULER_TRANSFORM, parent=jacket_l_front_01_root_ctrl_2, type=Control, loc=<0 0 0>, rot=<0 0 0 1>}
 head_ctrl {controlType=EULER_TRANSFORM, type=Control, loc=<0 -1.6873 152.0342>, rot=<0.0029 -0.7071 -0.0029 0.7071>}
]
variables [
 HoodDown {type=bool, default=False}
 HoodieUpHeadBack {type=FTransform, default=(Rotation=(X=0,Y=0,Z=0,W=1),Translation=(X=0,Y=1.5,Z=0),Scale3D=(X=1,Y=1,Z=1))}
 HoodieUpHeadFwd {type=FTransform, default=(Rotation=(X=0,Y=0,Z=0,W=1),Translation=(X=0,Y=1,Z=0),Scale3D=(X=0.3,Y=1,Z=1))}
 IsCrouching {type=bool, default=False}
 ScaleThighPoseReader {type=double, default=1}
 Weight {type=float, default=1}
 baseball_slide_jacket_c_rear {type=FTransform, default=(Rotation=(X=0,Y=0,Z=0,W=1),Translation=(X=0,Y=0,Z=0),Scale3D=(X=1,Y=1,Z=1))}
 baseball_slide_jacket_l_front {type=FTransform, default=(Rotation=(X=0,Y=0,Z=0,W=1),Translation=(X=0,Y=0,Z=0),Scale3D=(X=1,Y=1,Z=1))}
 baseball_slide_jacket_l_rear {type=FTransform, default=(Rotation=(X=0,Y=0,Z=0,W=1),Translation=(X=0,Y=0,Z=0),Scale3D=(X=1,Y=1,Z=1))}
 baseball_slide_jacket_r_front {type=FTransform, default=(Rotation=(X=0,Y=0,Z=0,W=1),Translation=(X=0,Y=0,Z=0),Scale3D=(X=1,Y=1,Z=1))}
 baseball_slide_jacket_r_rear {type=FTransform, default=(Rotation=(X=0,Y=0,Z=0,W=1),Translation=(X=0,Y=0,Z=0),Scale3D=(X=1,Y=1,Z=1))}
 baseball_sliding {type=bool, default=False}
 is_sliding {type=bool, default=False}
 jacket_front_leg_fwd {type=FTransform, default=(Rotation=(X=0.5217,Y=-0.4438,Z=-0.151,W=0.7128),Translation=(X=0,Y=0,Z=0),Scale3D=(X=1,Y=1,Z=1))}
 jacket_side_leg_fwd {type=FTransform, default=(Rotation=(X=0,Y=0,Z=0,W=1),Translation=(X=0,Y=0,Z=0),Scale3D=(X=1,Y=1,Z=1))}
 knee_slide_jacket_c_rear {type=FTransform, default=(Rotation=(X=0,Y=0,Z=0,W=1),Translation=(X=0,Y=0,Z=0),Scale3D=(X=1,Y=1,Z=1))}
 knee_slide_jacket_l_front {type=FTransform, default=(Rotation=(X=0,Y=0,Z=0,W=1),Translation=(X=0,Y=0,Z=0),Scale3D=(X=1,Y=1,Z=1))}
 knee_slide_jacket_l_rear {type=FTransform, default=(Rotation=(X=0,Y=0,Z=0,W=1),Translation=(X=0,Y=0,Z=0),Scale3D=(X=1,Y=1,Z=1))}
 knee_slide_jacket_r_front {type=FTransform, default=(Rotation=(X=0,Y=0,Z=0,W=1),Translation=(X=0,Y=0,Z=0),Scale3D=(X=1,Y=1,Z=1))}
 knee_slide_jacket_r_rear {type=FTransform, default=(Rotation=(X=0,Y=0,Z=0,W=1),Translation=(X=0,Y=0,Z=0),Scale3D=(X=1,Y=1,Z=1))}
 knee_sliding {type=bool, default=False}
 sackbag_thigh_r_fwd {type=FTransform, default=(Rotation=(X=0,Y=0,Z=0,W=1),Translation=(X=0,Y=0,Z=0),Scale3D=(X=1,Y=1,Z=1))}
 sackbag_thigh_r_fwd_normal {type=FTransform, default=(Rotation=(X=0,Y=0,Z=0,W=1),Translation=(X=0,Y=0,Z=0),Scale3D=(X=1,Y=1,Z=1))}
 sackbelt_hip_thigh_r_fwd {type=FTransform, default=(Rotation=(X=0,Y=0,Z=0,W=1),Translation=(X=0,Y=0,Z=0),Scale3D=(X=1,Y=1,Z=1))}
 sackbelt_hip_thigh_r_fwd_normal {type=FTransform, default=(Rotation=(X=0,Y=0,Z=0,W=1),Translation=(X=0,Y=0,Z=0),Scale3D=(X=1,Y=1,Z=1))}
 sackbelt_lowerback_thigh_r_fwd {type=FTransform, default=(Rotation=(X=0,Y=0,Z=0,W=1),Translation=(X=0,Y=0,Z=0),Scale3D=(X=1,Y=1,Z=1))}
 sackbelt_lowerback_thigh_r_fwd_normal {type=FTransform, default=(Rotation=(X=0,Y=0,Z=0,W=1),Translation=(X=0,Y=0,Z=0),Scale3D=(X=1,Y=1,Z=1))}
]
