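# Physics asset for the ZirconSweep medium male player body's jacket, providing collision and joints for secondary motion.
# Box and capsule bodies form front, side and rear panel chains from root bodies with limited swing and twist, plus extra right-side tail segments.
# Angularly free cross-constraints tie neighboring chains at lower segments, linking the panels like cloth.
# Pelvis, spine and leg capsules supply body shapes; sack bag and belt bodies attach to spine_01.
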PhysicsAsset /BRCosmetics/Characters/Player/Male/Medium/Bodies/M_MED_ZirconSweep/Meshes/M_MED_ZirconSweep_Jacket_Physics
bodies {
 calf_l [{shapeType=Capsule, length=23.6536, radius=8.9883, shapeName=calf_l_capsule, center=<-19.8712 -2.5099 0>, rotation=<rot 90 0 0>}]
 calf_r [{shapeType=Capsule, length=23.6536, radius=8.9883, shapeName=calf_l_capsule, center=<19.8712 2.5099 0>, rotation=<rot 90 0 180>}]
 dyn_sackbag [{
   shapeType=Capsule
   length=14.6644
   radius=6.1944
   shapeName=dyn_sackbag_capsule
   center=<-9.8877 1.9842 0.813>
   rotation=<rot 0.2019 94.3762 -88.7498>
  }]
 dyn_sackbelt_r_hip [{
   shapeType=Capsule
   length=15.021
   radius=3.6531
   shapeName=dyn_sackbelt_r_hip_capsule
   center=<-3.471 3.3332 6.4299>
   rotation=<rot -41.2331 -14.4119 -90.0535>
  }]
 jacket_c_rear_01 [{shapeType=Box, extentX=8.8125, extentY=16.5, extentZ=5.1875, shapeName=jacket_c_rear_01_box, center=<5.7504 0 0>}]
 jacket_c_rear_01_root [{shapeType=Capsule, length=10, radius=1, shapeName=jacket_c_rear_01_root_capsule, rotation=<rot 0 0 90>}]
 jacket_c_rear_02 [{shapeType=Capsule, length=3.4776, radius=4.1923, shapeName=jacket_c_rear_02_capsule, center=<1.8412 0 0>, rotation=<rot -90 90 -90>}]
 jacket_c_rear_03 [{shapeType=Capsule, length=8.3298, radius=4.2527, shapeName=jacket_c_rear_03_capsule, center=<3.2562 0 0>, rotation=<rot -90 180 -180>}]
 jacket_c_rear_04 [{shapeType=Capsule, length=8.1124, radius=3.9204, shapeName=jacket_c_rear_04_capsule, center=<4.6344 0 0>, rotation=<rot -90 90 -90>}]
 jacket_c_rear_05 [{shapeType=Capsule, length=7.1138, radius=5.0549, shapeName=jacket_c_rear_05_capsule, center=<10.1293 0 0>, rotation=<rot 90 -90 -90>}]
 jacket_l_front_01 [{
   shapeType=Box
   extentX=11.959
   extentY=19.3998
   extentZ=5.7005
   shapeName=jacket_l_front_01_box
   center=<5.6426 0.3996 1.1269>
   rotation=<rot 0.0007 0.0004 25.3124>
  }]
 jacket_l_front_01_root [{shapeType=Capsule, length=10, radius=1, shapeName=jacket_l_front_01_root_capsule, rotation=<rot 0 0 90>}]
 jacket_l_front_02 [{
   shapeType=Box
   extentX=12.3537
   extentY=18.9436
   extentZ=6.6997
   shapeName=jacket_l_front_02_box
   center=<7.5589 0.3791 1.5133>
   rotation=<rot 0.0004 0.0002 11.2499>
  }]
 jacket_l_front_03 [{
   shapeType=Box
   extentX=8.8125
   extentY=19.8073
   extentZ=6.3385
   shapeName=jacket_l_front_03_box
   center=<5.7993 0 1.5946>
   rotation=<rot -2.8121 0.0002 -0.0001>
  }]
 jacket_l_front_04 [{
   shapeType=Box
   extentX=8.8125
   extentY=21.9464
   extentZ=6.7837
   shapeName=jacket_l_front_04_box
   center=<5.8296 0 1.3527>
   rotation=<rot -3.3534 0 0>
  }]
 jacket_l_front_05 [{
   shapeType=Box
   extentX=20.5
   extentY=20.3457
   extentZ=7.2219
   shapeName=jacket_l_front_05_box
   center=<11.7119 0.501 0.4276>
   rotation=<rot -4.7751 0 0>
  }]
 jacket_l_rear_01 [{shapeType=Box, extentX=8.8125, extentY=16.5, extentZ=5.1875, shapeName=jacket_l_rear_01_box, center=<5.7504 0 0>}]
 jacket_l_rear_01_root [{shapeType=Capsule, length=10, radius=1, shapeName=jacket_l_rear_01_root_capsule, rotation=<rot 0 0 90>}]
 jacket_l_rear_02 [{
   shapeType=Box
   extentX=8.8125
   extentY=17.625
   extentZ=8.9454
   shapeName=jacket_l_rear_02_box
   center=<5.9114 -0.259 0.7716>
   rotation=<rot -3.7559 0.0002 0>
  }]
 jacket_l_rear_03 [{shapeType=Box, extentX=8.8125, extentY=16.5, extentZ=8.375, shapeName=jacket_l_rear_03_box, center=<5.9343 -0.259 0.9>}]
 jacket_l_rear_04 [{
   shapeType=Box
   extentX=8.8125
   extentY=16.5
   extentZ=8.25
   shapeName=jacket_l_rear_04_box
   center=<5.9343 -0.259 1.7314>
   rotation=<rot -4.1939 0 0>
  }]
 jacket_l_rear_05 [{
   shapeType=Capsule
   length=7.2448
   radius=5
   shapeName=jacket_l_rear_05_capsule
   center=<11.8509 -0.4444 0.5065>
   rotation=<rot -0.0254 0.0082 -91.7253>
  }]
 jacket_l_side_01 [{shapeType=Box, extentX=8.8125, extentY=16.5, extentZ=5.1875, shapeName=jacket_l_side_01_box, center=<5.7504 0 0>}]
 jacket_l_side_01_root [{shapeType=Capsule, length=10, radius=1, shapeName=jacket_l_side_01_root_capsule, rotation=<rot 0 0 90>}]
 jacket_l_side_02 [{
   shapeType=Box
   extentX=8.8125
   extentY=16.5
   extentZ=6.4696
   shapeName=jacket_l_side_02_box
   center=<5.8814 0 2.8398>
   rotation=<rot -1.5601 0.0573 1.8181>
  }]
 jacket_l_side_03 [{
   shapeType=Box
   extentX=8.8125
   extentY=19.55
   extentZ=6.8743
   shapeName=jacket_l_side_03_box
   center=<5.8795 0 2.5169>
   rotation=<rot -4.9514 0.0002 -0.6753>
  }]
 jacket_l_side_04 [{
   shapeType=Box
   extentX=9.2315
   extentY=19.7903
   extentZ=7.3461
   shapeName=jacket_l_side_04_box
   center=<6.7183 0 1.8161>
   rotation=<rot -7.7816 0.0003 -0.4324>
  }]
 jacket_l_side_05 [{
   shapeType=Box
   extentX=18.875
   extentY=21.3778
   extentZ=7.1207
   shapeName=jacket_l_side_05_box
   center=<11.8431 0 0.0283>
   rotation=<rot -7.2117 0 0>
  }]
 jacket_r_front_01 [{
   shapeType=Box
   extentX=10.4273
   extentY=18.4613
   extentZ=3.3868
   shapeName=jacket_l_front_01_box
   center=<-5.7473 -0.9379 0>
   rotation=<rot 0.0002 2.8126 180>
  }]
 jacket_r_front_01_root [{shapeType=Capsule, length=10, radius=1, shapeName=jacket_r_front_01_root_capsule, rotation=<rot 0 0 90>}]
 jacket_r_front_02 [{
   shapeType=Box
   extentX=8.8125
   extentY=20.4376
   extentZ=4.2873
   shapeName=jacket_l_front_02_box
   center=<-5.7888 -0.2541 -0.1752>
   rotation=<rot 2.8134 0.0004 179.9999>
  }]
 jacket_r_front_03 [{
   shapeType=Box
   extentX=8.8125
   extentY=19.966
   extentZ=5.065
   shapeName=jacket_l_front_03_box
   center=<-5.7959 -0.5518 -0.6713>
   rotation=<rot -3.8754 0 180>
  }]
 jacket_r_front_04 [{
   shapeType=Box
   extentX=8.8125
   extentY=20.0362
   extentZ=8.2409
   shapeName=jacket_l_front_04_box
   center=<-5.858 -0.5554 -1.3125>
   rotation=<rot -3.1747 -0.0798 177.2535>
  }]
 jacket_r_front_05 [{
   shapeType=Box
   extentX=20.5
   extentY=20.3457
   extentZ=9.5344
   shapeName=jacket_l_front_05_box
   center=<-11.7173 -0.501 -0.8124>
   rotation=<rot -3.2903 0 180>
  }]
 jacket_r_rear_01 [{shapeType=Box, extentX=8.8125, extentY=16.5, extentZ=5.1875, shapeName=jacket_l_rear_01_box, center=<-5.7504 0 0>, rotation=<rot 0 0 -180>}]
 jacket_r_rear_01_root [{shapeType=Capsule, length=10, radius=1, shapeName=jacket_r_rear_01_root_capsule, rotation=<rot 0 0 90>}]
 jacket_r_rear_02 [{
   shapeType=Box
   extentX=8.8125
   extentY=17.625
   extentZ=9.1954
   shapeName=jacket_l_rear_02_box
   center=<-5.9882 0.259 -0.9833>
   rotation=<rot -2.8778 0.0003 179.9999>
  }]
 jacket_r_rear_03 [{
   shapeType=Box
   extentX=8.8125
   extentY=16.5
   extentZ=9.5
   shapeName=jacket_l_rear_03_box
   center=<-5.949 0.259 -0.8993>
   rotation=<rot -5.6037 0 180>
  }]
 jacket_r_rear_04 [{
   shapeType=Box
   extentX=8.8125
   extentY=16.5
   extentZ=8.25
   shapeName=jacket_l_rear_04_box
   center=<-5.9343 0.259 -1.7314>
   rotation=<rot 0 0 -180>
  }]
 jacket_r_rear_05 [{
   shapeType=Capsule
   length=11.6245
   radius=5.455
   shapeName=jacket_r_rear_05_capsule
   center=<-12.8694 1.7966 -0.4852>
   rotation=<rot -0.0007 0.0002 81.5624>
  }]
 jacket_r_rear_06 [{
   shapeType=Capsule
   length=5.3066
   radius=7.4767
   shapeName=jacket_r_rear_06_capsule
   center=<-4.6264 -1.7006 -0.2207>
   rotation=<rot 5.3154 85.2987 -84.7857>
  }]
 jacket_r_rear_07 [{
   shapeType=Capsule
   length=6.2069
   radius=4.6048
   shapeName=jacket_r_rear_07_capsule
   center=<-4.0549 -1.8281 -0.1888>
   rotation=<rot 5.1343 -78.0099 -83.4993>
  }]
 jacket_r_side_01 [{
   shapeType=Box
   extentX=8.8125
   extentY=16.5
   extentZ=5.1875
   shapeName=jacket_l_side_01_box
   center=<-5.7504 0 -1.5011>
   rotation=<rot 0 0 -180>
  }]
 jacket_r_side_01_root [{shapeType=Capsule, length=10, radius=1, shapeName=jacket_r_side_01_root_capsule, rotation=<rot 0 0 90>}]
 jacket_r_side_02 [{
   shapeType=Box
   extentX=16.3962
   extentY=16.5
   extentZ=4.9169
   shapeName=jacket_l_side_02_box
   center=<-5.0488 0 -1.9043>
   rotation=<rot 0 0 -180>
  }]
 jacket_r_side_03 [{
   shapeType=Box
   extentX=8.8125
   extentY=16.5
   extentZ=7.5015
   shapeName=jacket_l_side_03_box
   center=<-5.8388 0 -2.5116>
   rotation=<rot -6.5481 0.0004 179.9999>
  }]
 jacket_r_side_04 [{
   shapeType=Box
   extentX=8.8125
   extentY=19.7903
   extentZ=9.5095
   shapeName=jacket_l_side_04_box
   center=<-5.8332 0 -1.4653>
   rotation=<rot -7.9415 0.0001 180>
  }]
 jacket_r_side_05 [{
   shapeType=Box
   extentX=18.875
   extentY=21.625
   extentZ=9.0504
   shapeName=jacket_l_side_05_box
   center=<-10.6771 0 0.2886>
   rotation=<rot -8.0319 0.0001 180>
  }]
 jacket_r_side_06 [{
   shapeType=Capsule
   length=8.1801
   radius=3.7753
   shapeName=jacket_r_side_06_capsule
   center=<-3.5317 0.5227 0.5204>
   rotation=<rot 5.5551 81.0981 -82.7658>
  }]
 pelvis [
  {shapeType=Capsule, length=12.2551, radius=10, shapeName=pelvis_capsule, center=<-2.7045 -1.5902 0>}
  {shapeType=Capsule, length=8.5193, radius=5.0729, shapeName=pelvis_capsule, center=<-8.7217 7.1801 0.0675>, rotation=<rot -89.5569 180 180>}
 ]
 spine_01 [{shapeType=Capsule, length=10.8441, radius=7.375, shapeName=spine_01_capsule, center=<3.6666 -0.4591 0>}]
 thigh_l [{
   shapeType=Capsule
   length=42.0132
   radius=9.8305
   shapeName=thigh_l_capsule
   center=<-21.9807 -1.3765 0.0751>
   rotation=<rot -86.6027 -0.0124 0.588>
  }]
 thigh_r [{
   shapeType=Capsule
   length=41.4507
   radius=10.393
   shapeName=thigh_l_capsule
   center=<21.9807 1.3765 -0.0751>
   rotation=<rot -86.6027 -0.0124 -179.412>
  }]
}
constraints [
 {bone1Name=jacket_c_rear_01, bone2Name=jacket_c_rear_01_root, swing1LimitDegrees=8, swing2LimitDegrees=12, twistLimitDegrees=4}
 {bone1Name=jacket_c_rear_02, bone2Name=jacket_c_rear_01, swing1LimitDegrees=8, swing2LimitDegrees=12, twistLimitDegrees=4}
 {bone1Name=jacket_c_rear_03, bone2Name=jacket_c_rear_02, swing1LimitDegrees=8, swing2LimitDegrees=12, twistLimitDegrees=4}
 {bone1Name=jacket_c_rear_04, bone2Name=jacket_c_rear_03, swing1LimitDegrees=12, swing2LimitDegrees=40, twistLimitDegrees=12}
 {bone1Name=jacket_c_rear_05, bone2Name=jacket_c_rear_04, swing1LimitDegrees=12, swing2LimitDegrees=40, twistLimitDegrees=12}
 {bone1Name=jacket_l_front_01, bone2Name=jacket_l_front_01_root, swing1LimitDegrees=20, swing2LimitDegrees=20, twistLimitDegrees=5}
 {bone1Name=jacket_l_front_02, bone2Name=jacket_l_front_01, swing1LimitDegrees=35, swing2LimitDegrees=30, twistLimitDegrees=30}
 {bone1Name=jacket_l_front_03, bone2Name=jacket_l_front_02, swing1LimitDegrees=35, swing2LimitDegrees=30, twistLimitDegrees=30}
 {bone1Name=jacket_l_front_04, bone2Name=jacket_l_front_03, swing1LimitDegrees=35, swing2LimitDegrees=20, twistLimitDegrees=20}
 {bone1Name=jacket_l_rear_01, bone2Name=jacket_l_rear_01_root, swing1LimitDegrees=8, swing2LimitDegrees=12, twistLimitDegrees=4}
 {bone1Name=jacket_l_rear_02, bone2Name=jacket_l_rear_01, swing1LimitDegrees=8, swing2LimitDegrees=12, twistLimitDegrees=4}
 {bone1Name=jacket_l_rear_03, bone2Name=jacket_l_rear_02, swing1LimitDegrees=8, swing2LimitDegrees=12, twistLimitDegrees=4}
 {bone1Name=jacket_l_rear_04, bone2Name=jacket_l_rear_03, swing1LimitDegrees=12, swing2LimitDegrees=40, twistLimitDegrees=12}
 {bone1Name=jacket_l_rear_05, bone2Name=jacket_l_rear_04, swing1LimitDegrees=12, swing2LimitDegrees=40, twistLimitDegrees=12}
 {bone1Name=jacket_l_side_01, bone2Name=jacket_l_side_01_root, swing1LimitDegrees=4, swing2LimitDegrees=4, twistLimitDegrees=4}
 {bone1Name=jacket_l_side_02, bone2Name=jacket_l_side_01, swing1LimitDegrees=16, swing2LimitDegrees=16, twistLimitDegrees=8}
 {bone1Name=jacket_l_side_03, bone2Name=jacket_l_side_02, swing1LimitDegrees=16, swing2LimitDegrees=16, twistLimitDegrees=8}
 {bone1Name=jacket_l_side_04, bone2Name=jacket_l_side_03, swing1LimitDegrees=24, swing2LimitDegrees=16, twistLimitDegrees=8}
 {bone1Name=jacket_l_side_05, bone2Name=jacket_l_side_04, swing1LimitDegrees=32, swing2LimitDegrees=16, twistLimitDegrees=8}
 {
  bone1Name=calf_r
  bone2Name=thigh_r
  swing1LimitDegrees=45
  swing1Motion=Free
  swing2LimitDegrees=45
  swing2Motion=Free
  twistLimitDegrees=45
  twistMotion=Free
 }
 {bone1Name=jacket_r_front_01, bone2Name=jacket_r_front_01_root, swing1LimitDegrees=5, swing2LimitDegrees=5, twistLimitDegrees=5}
 {bone1Name=jacket_r_front_02, bone2Name=jacket_r_front_01, swing1LimitDegrees=50, swing2LimitDegrees=20, twistLimitDegrees=20}
 {bone1Name=jacket_r_front_03, bone2Name=jacket_r_front_02, swing1LimitDegrees=50, swing2LimitDegrees=20, twistLimitDegrees=20}
 {bone1Name=jacket_r_front_04, bone2Name=jacket_r_front_03, swing1LimitDegrees=36, swing2LimitDegrees=20, twistLimitDegrees=20}
 {bone1Name=jacket_r_rear_01, bone2Name=jacket_r_rear_01_root, swing1LimitDegrees=8, swing2LimitDegrees=12, twistLimitDegrees=4}
 {bone1Name=jacket_r_rear_02, bone2Name=jacket_r_rear_01, swing1LimitDegrees=8, swing2LimitDegrees=12, twistLimitDegrees=4}
 {bone1Name=jacket_r_rear_03, bone2Name=jacket_r_rear_02, swing1LimitDegrees=8, swing2LimitDegrees=12, twistLimitDegrees=4}
 {bone1Name=jacket_r_rear_04, bone2Name=jacket_r_rear_03, swing1LimitDegrees=12, swing2LimitDegrees=40, twistLimitDegrees=12}
 {bone1Name=jacket_r_rear_05, bone2Name=jacket_r_rear_04, swing1LimitDegrees=12, swing2LimitDegrees=40, twistLimitDegrees=12}
 {bone1Name=jacket_r_side_01, bone2Name=jacket_r_side_01_root, swing1LimitDegrees=4, swing2LimitDegrees=4, twistLimitDegrees=4}
 {bone1Name=jacket_r_side_02, bone2Name=jacket_r_side_01, swing1LimitDegrees=16, swing2LimitDegrees=16, twistLimitDegrees=8}
 {bone1Name=jacket_r_side_03, bone2Name=jacket_r_side_02, swing1LimitDegrees=16, swing2LimitDegrees=16, twistLimitDegrees=8}
 {bone1Name=jacket_r_side_04, bone2Name=jacket_r_side_03, swing1LimitDegrees=8, swing2LimitDegrees=8, twistLimitDegrees=4}
 {bone1Name=jacket_r_front_05, bone2Name=jacket_r_front_04, swing1LimitDegrees=20, swing2LimitDegrees=20, twistLimitDegrees=4}
 {bone1Name=jacket_r_side_05, bone2Name=jacket_r_side_04, swing1LimitDegrees=24, swing2LimitDegrees=24, twistLimitDegrees=24}
 {
  bone1Name=jacket_c_rear_04
  bone2Name=jacket_l_rear_04
  swing1LimitDegrees=45
  swing1Motion=Free
  swing2LimitDegrees=45
  swing2Motion=Free
  twistLimitDegrees=45
  twistMotion=Free
 }
 {
  bone1Name=jacket_c_rear_04
  bone2Name=jacket_r_rear_04
  swing1LimitDegrees=45
  swing1Motion=Free
  swing2LimitDegrees=45
  swing2Motion=Free
  twistLimitDegrees=45
  twistMotion=Free
 }
 {
  bone1Name=jacket_r_side_04
  bone2Name=jacket_r_front_04
  swing1LimitDegrees=45
  swing1Motion=Free
  swing2LimitDegrees=45
  swing2Motion=Free
  twistLimitDegrees=45
  twistMotion=Free
 }
 {
  bone1Name=jacket_c_rear_05
  bone2Name=jacket_l_rear_05
  swing1LimitDegrees=45
  swing1Motion=Free
  swing2LimitDegrees=45
  swing2Motion=Free
  twistLimitDegrees=45
  twistMotion=Free
 }
 {
  bone1Name=dyn_sackbelt_r_hip
  bone2Name=spine_01
  swing1LimitDegrees=45
  swing1Motion=Free
  swing2LimitDegrees=45
  swing2Motion=Free
  twistLimitDegrees=45
  twistMotion=Free
 }
 {
  bone1Name=dyn_sackbag
  bone2Name=spine_01
  swing1LimitDegrees=45
  swing1Motion=Free
  swing2LimitDegrees=45
  swing2Motion=Free
  twistLimitDegrees=45
  twistMotion=Free
 }
 {bone1Name=jacket_r_rear_07, bone2Name=jacket_r_rear_05, swing1LimitDegrees=24, swing2LimitDegrees=24, twistLimitDegrees=24}
 {bone1Name=jacket_r_rear_06, bone2Name=jacket_r_rear_05, swing1LimitDegrees=24, swing2LimitDegrees=24, twistLimitDegrees=24}
 {bone1Name=jacket_r_side_06, bone2Name=jacket_r_side_05, swing1LimitDegrees=24, swing2LimitDegrees=24, twistLimitDegrees=24}
 {
  bone1Name=thigh_r
  bone2Name=pelvis
  swing1LimitDegrees=45
  swing1Motion=Free
  swing2LimitDegrees=45
  swing2Motion=Free
  twistLimitDegrees=45
  twistMotion=Free
 }
 {
  bone1Name=jacket_r_rear_05
  bone2Name=jacket_c_rear_05
  swing1LimitDegrees=45
  swing1Motion=Free
  swing2LimitDegrees=45
  swing2Motion=Free
  twistLimitDegrees=45
  twistMotion=Free
 }
 {
  bone1Name=jacket_r_side_04
  bone2Name=jacket_r_rear_04
  swing1LimitDegrees=45
  swing1Motion=Free
  swing2LimitDegrees=45
  swing2Motion=Free
  twistLimitDegrees=45
  twistMotion=Free
 }
 {
  bone1Name=jacket_l_side_04
  bone2Name=jacket_l_rear_04
  swing1LimitDegrees=45
  swing1Motion=Free
  swing2LimitDegrees=45
  swing2Motion=Free
  twistLimitDegrees=45
  twistMotion=Free
 }
 {
  bone1Name=jacket_l_front_04
  bone2Name=jacket_l_side_04
  swing1LimitDegrees=45
  swing1Motion=Free
  swing2LimitDegrees=45
  swing2Motion=Free
  twistLimitDegrees=45
  twistMotion=Free
 }
 {
  bone1Name=jacket_l_front_03
  bone2Name=jacket_l_side_03
  swing1LimitDegrees=45
  swing1Motion=Free
  swing2LimitDegrees=45
  swing2Motion=Free
  twistLimitDegrees=45
  twistMotion=Free
 }
 {
  bone1Name=thigh_l
  bone2Name=pelvis
  swing1LimitDegrees=45
  swing1Motion=Free
  swing2LimitDegrees=45
  swing2Motion=Free
  twistLimitDegrees=45
  twistMotion=Free
 }
 {bone1Name=jacket_l_front_05, bone2Name=jacket_l_front_04, swing1LimitDegrees=32, swing2LimitDegrees=32, twistLimitDegrees=32}
 {
  bone1Name=jacket_l_front_05
  bone2Name=jacket_l_side_05
  swing1LimitDegrees=45
  swing1Motion=Free
  swing2LimitDegrees=45
  swing2Motion=Free
  twistLimitDegrees=45
  twistMotion=Free
 }
 {bone1Name=calf_l, bone2Name=thigh_l, swing1LimitDegrees=10, swing2LimitDegrees=10, twistLimitDegrees=10}
 {
  bone1Name=jacket_r_side_05
  bone2Name=jacket_r_front_05
  swing1LimitDegrees=45
  swing1Motion=Free
  swing2LimitDegrees=45
  swing2Motion=Free
  twistLimitDegrees=45
  twistMotion=Free
 }
]
