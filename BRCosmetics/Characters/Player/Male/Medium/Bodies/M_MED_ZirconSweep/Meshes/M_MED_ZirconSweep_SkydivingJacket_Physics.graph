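# Physics asset for the ZirconSweep medium male character's skydiving jacket, defining bodies and joints for jacket motion.
# Seven box and capsule chains (front, side, rear per side, plus center rear) descend from root capsules with swing and twist limits.
# Free-angular constraints cross-link neighboring lower panels, tying them together like cloth, with extra tip segments on the right.
# Pelvis, spine and leg capsules provide body shapes, and a sack bag and hip belt joint to spine_01.
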PhysicsAsset /BRCosmetics/Characters/Player/Male/Medium/Bodies/M_MED_ZirconSweep/Meshes/M_MED_ZirconSweep_SkydivingJacket_Physics
bodies {
 calf_l [{shapeType=Capsule, length=19.2786, radius=8.4883, shapeName=calf_l_capsule, center=<-14.9064 -2.5099 0>, rotation=<rot 90 0 0>}]
 calf_r [{shapeType=Capsule, length=23.6536, radius=8.9883, shapeName=calf_l_capsule, center=<19.8712 2.5099 0>, rotation=<rot 90 0 180>}]
 dyn_sackbag [{
   shapeType=Capsule
   length=11.7083
   radius=7.9443
   shapeName=dyn_sackbag_capsule
   center=<-8.477 4.0013 0.8537>
   rotation=<rot 0.2011 94.3769 -88.7501>
  }]
 dyn_sackbelt_r_hip [{
   shapeType=Capsule
   length=15.1772
   radius=3.4969
   shapeName=dyn_sackbelt_r_hip_capsule
   center=<-5.1276 1.1358 3.8273>
   rotation=<rot -41.2331 -14.4119 -90.0535>
  }]
 jacket_c_rear_01 [{shapeType=Box, extentX=8.8125, extentY=16.5, extentZ=5.1875, shapeName=jacket_c_rear_01_box, center=<5.7504 0 0>}]
 jacket_c_rear_01_root [{shapeType=Capsule, length=10, radius=1, shapeName=jacket_c_rear_01_root_capsule, rotation=<rot 0 0 90>}]
 jacket_c_rear_02 [{shapeType=Capsule, length=2.4776, radius=5.1923, shapeName=jacket_c_rear_02_capsule, center=<1.8412 0 0>, rotation=<rot -90 90 -90>}]
 jacket_c_rear_03 [{shapeType=Capsule, length=7.3298, radius=5.2527, shapeName=jacket_c_rear_03_capsule, center=<3.2562 0 0>, rotation=<rot -90 180 -180>}]
 jacket_c_rear_04 [{shapeType=Capsule, length=5.1124, radius=6.9204, shapeName=jacket_c_rear_04_capsule, center=<4.6344 0 0>, rotation=<rot -90 90 -90>}]
 jacket_c_rear_05 [{shapeType=Capsule, length=7.1138, radius=5.0549, shapeName=jacket_c_rear_05_capsule, center=<10.1293 0 0>, rotation=<rot 90 -90 -90>}]
 jacket_l_front_01 [{shapeType=Box, extentX=8.8125, extentY=18.5812, extentZ=5.1875, shapeName=jacket_l_front_01_box, center=<5.7504 0 0>}]
 jacket_l_front_01_root [{shapeType=Capsule, length=10, radius=1, shapeName=jacket_l_front_01_root_capsule, rotation=<rot 0 0 90>}]
 jacket_l_front_02 [{shapeType=Box, extentX=8.8125, extentY=18.7199, extentZ=5.1875, shapeName=jacket_l_front_02_box, center=<5.7504 0 0>}]
 jacket_l_front_03 [{shapeType=Box, extentX=8.8125, extentY=18.9934, extentZ=6.3385, shapeName=jacket_l_front_03_box, center=<5.7504 0 0>}]
 jacket_l_front_04 [{shapeType=Box, extentX=8.8125, extentY=19.3135, extentZ=7.3385, shapeName=jacket_l_front_04_box, center=<5.7504 0 0>}]
 jacket_l_front_05 [{
   shapeType=Box
   extentX=20.5
   extentY=21.6987
   extentZ=8.5885
   shapeName=jacket_l_front_05_box
   center=<11.5345 -1.0989 0.439>
   rotation=<rot -2.0472 0.0031 -0.362>
  }]
 jacket_l_rear_01 [{shapeType=Box, extentX=8.8125, extentY=16.5, extentZ=5.1875, shapeName=jacket_l_rear_01_box, center=<5.7504 0 0>}]
 jacket_l_rear_01_root [{shapeType=Capsule, length=10, radius=1, shapeName=jacket_l_rear_01_root_capsule, rotation=<rot 0 0 90>}]
 jacket_l_rear_02 [{
   shapeType=Box
   extentX=8.8125
   extentY=17.625
   extentZ=9.7621
   shapeName=jacket_l_rear_02_box
   center=<5.9269 -0.259 1.0075>
   rotation=<rot -3.7559 0.0002 0>
  }]
 jacket_l_rear_03 [{shapeType=Box, extentX=8.8125, extentY=16.5, extentZ=9.8882, shapeName=jacket_l_rear_03_box, center=<5.9343 -0.259 0.9>}]
 jacket_l_rear_04 [{shapeType=Box, extentX=8.8125, extentY=16.5, extentZ=8.25, shapeName=jacket_l_rear_04_box, center=<5.9343 -0.259 1.7314>}]
 jacket_l_rear_05 [{
   shapeType=Capsule
   length=4.6875
   radius=8.5
   shapeName=jacket_l_rear_05_capsule
   center=<11.8509 -0.0151 0.5194>
   rotation=<rot -0.0254 0.0082 -91.7253>
  }]
 jacket_l_side_01 [{shapeType=Box, extentX=8.8125, extentY=16.5, extentZ=5.1875, shapeName=jacket_l_side_01_box, center=<5.7504 0 0>}]
 jacket_l_side_01_root [{shapeType=Capsule, length=10, radius=1, shapeName=jacket_l_side_01_root_capsule, rotation=<rot 0 0 90>}]
 jacket_l_side_02 [{shapeType=Box, extentX=8.8125, extentY=16.5, extentZ=6.2662, shapeName=jacket_l_side_02_box, center=<5.7504 0 0.9865>}]
 jacket_l_side_03 [{shapeType=Box, extentX=8.8125, extentY=16.5, extentZ=5.1875, shapeName=jacket_l_side_03_box, center=<5.7504 0 0.8254>}]
 jacket_l_side_04 [{
   shapeType=Box
   extentX=8.8125
   extentY=16.5
   extentZ=5.1875
   shapeName=jacket_l_side_04_box
   center=<5.7418 0 0.5422>
   rotation=<rot -2.4951 0.0001 0>
  }]
 jacket_l_side_05 [{shapeType=Box, extentX=18.875, extentY=16.5, extentZ=5.1875, shapeName=jacket_l_side_05_box, center=<10.8254 0 0.3364>}]
 jacket_r_front_01 [{shapeType=Box, extentX=8.8125, extentY=16.5, extentZ=3.3125, shapeName=jacket_l_front_01_box, center=<-5.7504 0 0>, rotation=<rot 0 0 -180>}]
 jacket_r_front_01_root [{shapeType=Capsule, length=10, radius=1, shapeName=jacket_r_front_01_root_capsule, rotation=<rot 0 0 90>}]
 jacket_r_front_02 [{
   shapeType=Box
   extentX=8.8125
   extentY=16.5
   extentZ=4.3662
   shapeName=jacket_l_front_02_box
   center=<-5.7504 0 0>
   rotation=<rot 5.6252 0.0001 180>
  }]
 jacket_r_front_03 [{
   shapeType=Box
   extentX=8.8125
   extentY=19.0466
   extentZ=6.7451
   shapeName=jacket_l_front_03_box
   center=<-5.7504 0 0>
   rotation=<rot 0 0 -180>
  }]
 jacket_r_front_04 [{
   shapeType=Box
   extentX=8.8125
   extentY=18.9061
   extentZ=6.8716
   shapeName=jacket_l_front_04_box
   center=<-5.7504 0 0>
   rotation=<rot 0 0 -180>
  }]
 jacket_r_front_05 [{
   shapeType=Box
   extentX=20.5
   extentY=19.7581
   extentZ=7.9375
   shapeName=jacket_l_front_05_box
   center=<-11.6706 0 0>
   rotation=<rot -3.9754 0 180>
  }]
 jacket_r_rear_01 [{shapeType=Box, extentX=8.8125, extentY=16.5, extentZ=5.1875, shapeName=jacket_l_rear_01_box, center=<-5.7504 0 0>, rotation=<rot 0 0 -180>}]
 jacket_r_rear_01_root [{shapeType=Capsule, length=10, radius=1, shapeName=jacket_r_rear_01_root_capsule, rotation=<rot 0 0 90>}]
 jacket_r_rear_02 [{
   shapeType=Box
   extentX=8.8125
   extentY=17.625
   extentZ=10.0121
   shapeName=jacket_l_rear_02_box
   center=<-6.0001 0.259 -1.2193>
   rotation=<rot -2.8778 0.0003 179.9999>
  }]
 jacket_r_rear_03 [{
   shapeType=Box
   extentX=8.8125
   extentY=16.5
   extentZ=11.0132
   shapeName=jacket_l_rear_03_box
   center=<-5.949 0.259 -0.8993>
   rotation=<rot -9.8777 0.0001 180>
  }]
 jacket_r_rear_04 [{
   shapeType=Box
   extentX=8.8125
   extentY=16.5
   extentZ=9.25
   shapeName=jacket_l_rear_04_box
   center=<-5.9343 0.259 -1.7314>
   rotation=<rot -9.5626 0 180>
  }]
 jacket_r_rear_05 [{
   shapeType=Box
   extentX=17.1875
   extentY=17.125
   extentZ=12
   shapeName=jacket_l_rear_05_box
   center=<-10.1197 2.548 -1.5682>
   rotation=<rot 0 0 -180>
  }]
 jacket_r_rear_06 [{
   shapeType=Capsule
   length=5.3066
   radius=7.4767
   shapeName=jacket_r_rear_06_capsule
   center=<-3.0358 -1.4349 -0.0406>
   rotation=<rot 5.3154 85.2987 -84.7857>
  }]
 jacket_r_rear_07 [{
   shapeType=Capsule
   length=6.2069
   radius=4.6048
   shapeName=jacket_r_rear_07_capsule
   center=<-4.0549 -1.8281 -0.1888>
   rotation=<rot 5.1343 -78.0099 -83.4993>
  }]
 jacket_r_side_01 [{
   shapeType=Box
   extentX=8.8125
   extentY=16.5
   extentZ=5.1875
   shapeName=jacket_l_side_01_box
   center=<-5.7504 0 -1.5011>
   rotation=<rot 0 0 -180>
  }]
 jacket_r_side_01_root [{shapeType=Capsule, length=10, radius=1, shapeName=jacket_r_side_01_root_capsule, rotation=<rot 0 0 90>}]
 jacket_r_side_02 [{
   shapeType=Box
   extentX=14.0625
   extentY=15.9166
   extentZ=4.5298
   shapeName=jacket_l_side_02_box
   center=<-5.0913 0 -1.4727>
   rotation=<rot 5.625 0.0001 180>
  }]
 jacket_r_side_03 [{
   shapeType=Box
   extentX=8.8125
   extentY=16.5
   extentZ=5.9375
   shapeName=jacket_l_side_03_box
   center=<-5.7205 0 -1.1226>
   rotation=<rot -9.9494 0.0005 179.9999>
  }]
 jacket_r_side_04 [{
   shapeType=Box
   extentX=8.8125
   extentY=16.5
   extentZ=6.1875
   shapeName=jacket_l_side_04_box
   center=<-5.7504 0 -1.2342>
   rotation=<rot -8.3854 0.0001 180>
  }]
 jacket_r_side_05 [{
   shapeType=Box
   extentX=18.875
   extentY=21.625
   extentZ=7.9375
   shapeName=jacket_l_side_05_box
   center=<-10.7968 0 -0.5633>
   rotation=<rot -10.603 0.0002 179.9999>
  }]
 jacket_r_side_06 [{
   shapeType=Capsule
   length=8.1801
   radius=3.7753
   shapeName=jacket_r_side_06_capsule
   center=<-3.5317 0.5227 0.5204>
   rotation=<rot 5.5551 81.0981 -82.7658>
  }]
 pelvis [{shapeType=Capsule, length=13.0294, radius=9.9306, shapeName=pelvis_capsule, center=<-4.7169 -1.5902 0>}]
 spine_01 [{shapeType=Capsule, length=9.9419, radius=8.4298, shapeName=spine_01_capsule, center=<3.6666 -0.5532 0>}]
 thigh_l [{
   shapeType=Capsule
   length=42.1455
   radius=9.6983
   shapeName=thigh_l_capsule
   center=<-21.9807 -1.3765 0.0751>
   rotation=<rot -86.6027 -0.0124 0.588>
  }]
 thigh_r [{
   shapeType=Capsule
   length=41.8017
   radius=10.042
   shapeName=thigh_l_capsule
   center=<21.9807 1.3765 -0.0751>
   rotation=<rot -86.6027 -0.0124 -179.412>
  }]
}
constraints [
 {bone1Name=jacket_c_rear_01, bone2Name=jacket_c_rear_01_root, swing1LimitDegrees=8, swing2LimitDegrees=12, twistLimitDegrees=4}
 {bone1Name=jacket_c_rear_02, bone2Name=jacket_c_rear_01, swing1LimitDegrees=8, swing2LimitDegrees=12, twistLimitDegrees=4}
 {bone1Name=jacket_c_rear_03, bone2Name=jacket_c_rear_02, swing1LimitDegrees=8, swing2LimitDegrees=12, twistLimitDegrees=4}
 {bone1Name=jacket_c_rear_04, bone2Name=jacket_c_rear_03, swing1LimitDegrees=12, swing2LimitDegrees=40, twistLimitDegrees=12}
 {bone1Name=jacket_c_rear_05, bone2Name=jacket_c_rear_04, swing1LimitDegrees=12, swing2LimitDegrees=40, twistLimitDegrees=12}
 {bone1Name=jacket_l_front_01, bone2Name=jacket_l_front_01_root, swing1LimitDegrees=50, swing2LimitDegrees=20, twistLimitDegrees=20}
 {bone1Name=jacket_l_front_02, bone2Name=jacket_l_front_01, swing1LimitDegrees=50, swing2LimitDegrees=20, twistLimitDegrees=20}
 {bone1Name=jacket_l_front_03, bone2Name=jacket_l_front_02, swing1LimitDegrees=50, swing2LimitDegrees=20, twistLimitDegrees=20}
 {bone1Name=jacket_l_front_04, bone2Name=jacket_l_front_03, swing1LimitDegrees=50, swing2LimitDegrees=20, twistLimitDegrees=20}
 {bone1Name=jacket_l_rear_01, bone2Name=jacket_l_rear_01_root, swing1LimitDegrees=8, swing2LimitDegrees=12, twistLimitDegrees=4}
 {bone1Name=jacket_l_rear_02, bone2Name=jacket_l_rear_01, swing1LimitDegrees=8, swing2LimitDegrees=12, twistLimitDegrees=4}
 {bone1Name=jacket_l_rear_03, bone2Name=jacket_l_rear_02, swing1LimitDegrees=8, swing2LimitDegrees=12, twistLimitDegrees=4}
 {bone1Name=jacket_l_rear_04, bone2Name=jacket_l_rear_03, swing1LimitDegrees=12, swing2LimitDegrees=40, twistLimitDegrees=12}
 {bone1Name=jacket_l_rear_05, bone2Name=jacket_l_rear_04, swing1LimitDegrees=12, swing2LimitDegrees=40, twistLimitDegrees=12}
 {bone1Name=jacket_l_side_01, bone2Name=jacket_l_side_01_root, swing1LimitDegrees=4, swing2LimitDegrees=4, twistLimitDegrees=4}
 {bone1Name=jacket_l_side_02, bone2Name=jacket_l_side_01, swing1LimitDegrees=16, swing2LimitDegrees=16, twistLimitDegrees=8}
 {bone1Name=jacket_l_side_03, bone2Name=jacket_l_side_02, swing1LimitDegrees=16, swing2LimitDegrees=16, twistLimitDegrees=8}
 {bone1Name=jacket_l_side_04, bone2Name=jacket_l_side_03, swing1LimitDegrees=16, swing2LimitDegrees=16, twistLimitDegrees=8}
 {bone1Name=jacket_l_front_05, bone2Name=jacket_l_front_04, swing1LimitDegrees=20, swing2LimitDegrees=20, twistLimitDegrees=4}
 {bone1Name=jacket_l_side_05, bone2Name=jacket_l_side_04, swing1LimitDegrees=16, swing2LimitDegrees=16, twistLimitDegrees=8}
 {
  bone1Name=calf_l
  bone2Name=thigh_l
  swing1LimitDegrees=45
  swing1Motion=Free
  swing2LimitDegrees=45
  swing2Motion=Free
  twistLimitDegrees=45
  twistMotion=Free
 }
 {
  bone1Name=calf_r
  bone2Name=thigh_r
  swing1LimitDegrees=45
  swing1Motion=Free
  swing2LimitDegrees=45
  swing2Motion=Free
  twistLimitDegrees=45
  twistMotion=Free
 }
 {bone1Name=jacket_r_front_01, bone2Name=jacket_r_front_01_root, swing1LimitDegrees=50, swing2LimitDegrees=20, twistLimitDegrees=20}
 {bone1Name=jacket_r_front_02, bone2Name=jacket_r_front_01, swing1LimitDegrees=50, swing2LimitDegrees=20, twistLimitDegrees=20}
 {bone1Name=jacket_r_front_03, bone2Name=jacket_r_front_02, swing1LimitDegrees=20, swing2LimitDegrees=20, twistLimitDegrees=20}
 {bone1Name=jacket_r_front_04, bone2Name=jacket_r_front_03, swing1LimitDegrees=14, swing2LimitDegrees=14, twistLimitDegrees=4}
 {bone1Name=jacket_r_rear_01, bone2Name=jacket_r_rear_01_root, swing1LimitDegrees=8, swing2LimitDegrees=12, twistLimitDegrees=4}
 {bone1Name=jacket_r_rear_02, bone2Name=jacket_r_rear_01, swing1LimitDegrees=8, swing2LimitDegrees=12, twistLimitDegrees=4}
 {bone1Name=jacket_r_rear_03, bone2Name=jacket_r_rear_02, swing1LimitDegrees=8, swing2LimitDegrees=12, twistLimitDegrees=4}
 {bone1Name=jacket_r_rear_04, bone2Name=jacket_r_rear_03, swing1LimitDegrees=12, swing2LimitDegrees=40, twistLimitDegrees=12}
 {bone1Name=jacket_r_rear_05, bone2Name=jacket_r_rear_04, swing1LimitDegrees=12, swing2LimitDegrees=12, twistLimitDegrees=12}
 {bone1Name=jacket_r_side_01, bone2Name=jacket_r_side_01_root, swing1LimitDegrees=4, swing2LimitDegrees=4, twistLimitDegrees=4}
 {bone1Name=jacket_r_side_02, bone2Name=jacket_r_side_01, swing1LimitDegrees=16, swing2LimitDegrees=16, twistLimitDegrees=8}
 {bone1Name=jacket_r_side_03, bone2Name=jacket_r_side_02, swing1LimitDegrees=16, swing2LimitDegrees=16, twistLimitDegrees=8}
 {bone1Name=jacket_r_side_04, bone2Name=jacket_r_side_03, swing1LimitDegrees=16, swing2LimitDegrees=16, twistLimitDegrees=4}
 {bone1Name=jacket_r_front_05, bone2Name=jacket_r_front_04, swing1LimitDegrees=10, swing2LimitDegrees=10, twistLimitDegrees=4}
 {bone1Name=jacket_r_side_05, bone2Name=jacket_r_side_04, swing1LimitDegrees=32, swing2LimitDegrees=32, twistLimitDegrees=32}
 {
  bone1Name=jacket_r_rear_04
  bone2Name=jacket_r_side_04
  swing1LimitDegrees=45
  swing1Motion=Free
  swing2LimitDegrees=45
  swing2Motion=Free
  twistLimitDegrees=45
  twistMotion=Free
 }
 {
  bone1Name=jacket_r_rear_03
  bone2Name=jacket_r_side_03
  swing1LimitDegrees=45
  swing1Motion=Free
  swing2LimitDegrees=45
  swing2Motion=Free
  twistLimitDegrees=45
  twistMotion=Free
 }
 {
  bone1Name=jacket_r_side_04
  bone2Name=jacket_r_front_04
  swing1LimitDegrees=45
  swing1Motion=Free
  swing2LimitDegrees=45
  swing2Motion=Free
  twistLimitDegrees=45
  twistMotion=Free
 }
 {
  bone1Name=jacket_l_rear_04
  bone2Name=jacket_l_side_04
  swing1LimitDegrees=45
  swing1Motion=Free
  swing2LimitDegrees=45
  swing2Motion=Free
  twistLimitDegrees=45
  twistMotion=Free
 }
 {
  bone1Name=jacket_l_rear_03
  bone2Name=jacket_l_side_03
  swing1LimitDegrees=45
  swing1Motion=Free
  swing2LimitDegrees=45
  swing2Motion=Free
  twistLimitDegrees=45
  twistMotion=Free
 }
 {
  bone1Name=jacket_l_side_04
  bone2Name=jacket_l_front_04
  swing1LimitDegrees=45
  swing1Motion=Free
  swing2LimitDegrees=45
  swing2Motion=Free
  twistLimitDegrees=45
  twistMotion=Free
 }
 {
  bone1Name=jacket_c_rear_05
  bone2Name=jacket_l_rear_05
  swing1LimitDegrees=45
  swing1Motion=Free
  swing2LimitDegrees=45
  swing2Motion=Free
  twistLimitDegrees=45
  twistMotion=Free
 }
 {
  bone1Name=dyn_sackbelt_r_hip
  bone2Name=spine_01
  swing1LimitDegrees=45
  swing1Motion=Free
  swing2LimitDegrees=45
  swing2Motion=Free
  twistLimitDegrees=45
  twistMotion=Free
 }
 {
  bone1Name=dyn_sackbag
  bone2Name=spine_01
  swing1LimitDegrees=45
  swing1Motion=Free
  swing2LimitDegrees=45
  swing2Motion=Free
  twistLimitDegrees=45
  twistMotion=Free
 }
 {
  bone1Name=thigh_l
  bone2Name=pelvis
  swing1LimitDegrees=45
  swing1Motion=Free
  swing2LimitDegrees=45
  swing2Motion=Free
  twistLimitDegrees=45
  twistMotion=Free
 }
 {
  bone1Name=jacket_c_rear_05
  bone2Name=jacket_r_rear_05
  swing1LimitDegrees=45
  swing1Motion=Free
  swing2LimitDegrees=45
  swing2Motion=Free
  twistLimitDegrees=45
  twistMotion=Free
 }
 {
  bone1Name=jacket_r_rear_05
  bone2Name=jacket_r_side_05
  swing1LimitDegrees=45
  swing1Motion=Free
  swing2LimitDegrees=45
  swing2Motion=Free
  twistLimitDegrees=45
  twistMotion=Free
 }
 {bone1Name=jacket_r_rear_07, bone2Name=jacket_r_rear_05, swing1LimitDegrees=24, swing2LimitDegrees=24, twistLimitDegrees=24}
 {bone1Name=jacket_r_rear_06, bone2Name=jacket_r_rear_05, swing1LimitDegrees=24, swing2LimitDegrees=24, twistLimitDegrees=24}
 {bone1Name=jacket_r_side_06, bone2Name=jacket_r_side_05, swing1LimitDegrees=24, swing2LimitDegrees=24, twistLimitDegrees=24}
 {
  bone1Name=jacket_l_rear_05
  bone2Name=jacket_l_side_05
  swing1LimitDegrees=45
  swing1Motion=Free
  swing2LimitDegrees=45
  swing2Motion=Free
  twistLimitDegrees=45
  twistMotion=Free
 }
 {
  bone1Name=thigh_r
  bone2Name=pelvis
  swing1LimitDegrees=45
  swing1Motion=Free
  swing2LimitDegrees=45
  swing2Motion=Free
  twistLimitDegrees=45
  twistMotion=Free
 }
]
