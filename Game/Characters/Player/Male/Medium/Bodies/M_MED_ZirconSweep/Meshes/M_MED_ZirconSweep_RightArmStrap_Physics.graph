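# Physics asset for the right arm strap on the ZirconSweep medium male player body, providing bodies for the strap's dynamic bones.
# A capsule on the right upper arm anchors a chain of capsules running from the strap root through three segments.
# The root joint has free swing and twist, while each following joint has tight angular limits.
# Shape names carry left-arm labels, suggesting a copy or mirror of the left-side asset.
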PhysicsAsset /Game/Characters/Player/Male/Medium/Bodies/M_MED_ZirconSweep/Meshes/M_MED_ZirconSweep_RightArmStrap_Physics
bodies {
 dyn_r_arm_strap_01 [{shapeType=Capsule, length=1.7634, radius=1.1818, shapeName=dyn_l_arm_strap_01_capsule, center=<-2 0 0>, rotation=<rot 90 180 0>}]
 dyn_r_arm_strap_02 [{shapeType=Capsule, length=2.1024, radius=1.3517, shapeName=dyn_l_arm_strap_02_capsule, center=<-2 0 0>, rotation=<rot 90 180 0>}]
 dyn_r_arm_strap_03 [{shapeType=Capsule, length=3.1887, radius=1.6819, shapeName=dyn_l_arm_strap_03_capsule, center=<-2 0 0>, rotation=<rot 90 180 0>}]
 dyn_r_arm_strap_root [{shapeType=Capsule, length=0.505, radius=0.505, shapeName=dyn_l_arm_strap_root_capsule, rotation=<rot 0 -90 180>}]
 upperarm_r [{shapeType=Capsule, length=14.34, radius=6.8414, shapeName=upperarm_l_capsule, center=<-12.2831 0 0.9414>, rotation=<rot 90 0 180>}]
}
constraints [
 {
  bone1Name=dyn_r_arm_strap_root
  bone2Name=upperarm_r
  swing1LimitDegrees=45
  swing1Motion=Free
  swing2LimitDegrees=45
  swing2Motion=Free
  twistLimitDegrees=45
  twistMotion=Free
 }
 {bone1Name=dyn_r_arm_strap_01, bone2Name=dyn_r_arm_strap_root, swing1LimitDegrees=8, swing2LimitDegrees=8, twistLimitDegrees=8}
 {bone1Name=dyn_r_arm_strap_02, bone2Name=dyn_r_arm_strap_01, swing1LimitDegrees=8, swing2LimitDegrees=8, twistLimitDegrees=8}
 {bone1Name=dyn_r_arm_strap_03, bone2Name=dyn_r_arm_strap_02, swing1LimitDegrees=8, swing2LimitDegrees=8, twistLimitDegrees=8}
]
