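# Control Rig for the M_MED_ZirconSweep face accessory that procedurally shakes dynamic hair bones while skydiving.
# When `skydiving` is true, a loop walks a hardcoded bone array and applies a rotation to each through Modify Transforms.
# The RandomScaledRotation function builds that rotation from per-axis noise with randomized amplitude, using `speed scale` as noise speed.
# A half-weight quaternion slerp of that rotation is left unused.
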ControlRigBlueprint /BRCosmetics/Characters/Player/Male/Medium/Bodies/M_MED_ZirconSweep/Meshes/Parts/M_MED_ZirconSweep_FaceAcc_CtrlRig
graphs [
 New Function_ContainedGraph {nodes [
   n1 = RigVMFunctionEntryNode()
   Random = /Script/RigVM.RigVMFunction_RandomFloat(Duration=10, Maximum=15, Minimum=0, Seed=217) {is=struct RigVMUnitNode, template="Random::Execute(in Seed,in Minimum,in Maximum,in Duration,out Result)"}
   Multiply = &RigVMFunction_MathDoubleMul(B=-1, A=@Random.Result)
   Noise = /Script/RigVM.RigVMFunction_NoiseFloat(Frequency=0.5, Value=0.5, Speed=@n1.Scale, Minimum=@Multiply.Result, Maximum=@Random.Result) {is=struct RigVMUnitNode, template="Noise::Execute(in Value,in Speed,in Frequency,in Minimum,in Maximum,out Result)"}
   n2 = &RigVMFunction_RandomFloat(Duration=10, Maximum=20, Minimum=0, Seed=963) {template="Random::Execute(in Seed,in Minimum,in Maximum,in Duration,out Result)"}
   n3 = &RigVMFunction_RandomFloat(Duration=10, Maximum=15, Minimum=0, Seed=132) {template="Random::Execute(in Seed,in Minimum,in Maximum,in Duration,out Result)"}
   RigVMFunction_MathQuaternionFromEuler = &RigVMFunction_MathQuaternionFromEuler(
    Euler=(X=0,Y=0,Z=0)
    Euler.X=@Noise.Result
    Euler.Y=&RigVMFunction_NoiseFloat(Frequency=0.5, Value=0.3, Speed=@n1.Scale, Minimum=&RigVMFunction_MathDoubleMul(B=-1, A=@n2.Result).Result, Maximum=@n2.Result) {template="Noise::Execute(in Value,in Speed,in Frequency,in Minimum,in Maximum,out Result)"}.Result
    Euler.Z=&RigVMFunction_NoiseFloat(Frequency=0.5, Value=0.5, Speed=@n1.Scale, Minimum=&RigVMFunction_MathDoubleMul(B=-1, A=@n3.Result).Result, Maximum=@n3.Result) {template="Noise::Execute(in Value,in Speed,in Frequency,in Minimum,in Maximum,out Result)"}.Result
   )
   > RigVMFunctionReturnNode(Result=@RigVMFunction_MathQuaternionFromEuler.Result)
  ]}
 RigVMFunctionLibrary {nodes [RandomScaledRotation = RigVMCollapseNode(Scale=0)]}
 RigVMModel {nodes [
   RigUnit_BeginExecution = &RigUnit_BeginExecution()
   RigVMFunction_ControlFlowBranch = &RigVMFunction_ControlFlowBranch(ExecuteContext=@RigUnit_BeginExecution.ExecutePin, Condition=RigVMVariableNode() {variable=skydiving}.Value)
   RerouteNode = RigVMDispatchNode(Values="((Type=Bone,Name=\"dyn_hair_25\"),((),Type=Bone,Name=\"dyn_hair_24\"),((),Type=Bone,Name=\"dyn_hair_27\"),((),Type=Bone,Name=\"dyn_hair_28\"),((),Type=Bone,Name=\"dyn_hair_32\"),((),Type=Bone,Name=\"dyn_hair_33\"),((),Type=Bone,Name=\"dyn_hair_26\"),((),Type=Bone,Name=\" ...(truncated)") {template="DISPATCH_RigVMDispatch_ArrayMake(in Values,out Array)"}
   n1 = /Script/RigVM.RigVMFunction_ForLoopCount(
    Count=RigVMDispatchNode(Array=@RerouteNode.Array) {template="DISPATCH_RigVMDispatch_ArrayGetNum(in Array,out Num)"}.Num
    ExecutePin=@RigVMFunction_ControlFlowBranch.True
   ) {is=struct RigVMUnitNode}
   RandomScaledRotation = RigVMFunctionReferenceNode(Scale=RigVMVariableNode() {variable=speed scale}.Value)
   > &RigUnit_ModifyTransforms(
    ItemToModify=((Item=(Type=Bone),Transform=(Rotation=(X=0,Y=0,Z=0,W=1),Translation=(X=0,Y=0,Z=0),Scale3D=(X=1,Y=1,Z=1))))
    Weight=1
    ItemToModify.0.Item=RigVMDispatchNode(Index=@n1.Index, Array=@RerouteNode.Array) {template="DISPATCH_RigVMDispatch_ArrayGetAtIndex(in Array,in Index,out Element)"}.Element
    ItemToModify.0.Transform.Rotation=@RandomScaledRotation.Result
   )
   Interpolate = /Script/RigVM.RigVMFunction_MathQuaternionSlerp(A=(X=0,Y=0,Z=0,W=1), T=0.5, B=@RandomScaledRotation.Result) {is=struct RigVMUnitNode, template="Interpolate::Execute(in A,in B,in T,out Result)"}
  ]}
]
hierarchy [
 root {loc=<0 0 0>, rot=<0 0 0 1>}
 pelvis {parent=root, loc=<0 -0.7071 93.1191>, rot=<0.0127 0.707 -0.0127 -0.707>}
 spine_01 {parent=pelvis, loc=<6.7494 -0.6388 0>, rot=<0 0 0.034 -0.9994>}
 spine_02 {parent=spine_01, loc=<7.2518 -0.035 0>, rot=<0 0 0.0238 -0.9997>}
 spine_03 {parent=spine_02, loc=<6.0104 0 0>, rot=<0 0 0.0149 -0.9999>}
 spine_04 {parent=spine_03, loc=<8.1024 0 0>, rot=<0 0 -0.0132 -0.9999>}
 spine_05 {parent=spine_04, loc=<10.0146 0 0>, rot=<0 0 -0.0015 -1>}
 neck_01 {parent=spine_05, loc=<11.3138 0.1529 0>, rot=<0 0 -0.2003 -0.9797>}
 neck_02 {parent=neck_01, loc=<4.5627 0 0>, rot=<0 0 0.0504 -0.9987>}
 head {parent=neck_02, loc=<5.3914 0 0>, rot=<0 0 0.115 -0.9934>}
 dyn_hair_25 {parent=head, loc=<8.8807 10.1856 -2.4838>, rot=<0.3174 0.6261 0.6414 0.3097>}
 dyn_hair_24 {parent=head, loc=<8.9827 10.2564 -0.657>, rot=<0.1028 0.2474 0.9632 0.0227>}
 dyn_hair_27 {parent=head, loc=<7.6688 10.0526 4.2664>, rot=<0.1684 0.1368 0.976 0.0195>}
 dyn_hair_28 {parent=head, loc=<12.2766 9.7616 3.6532>, rot=<0.5166 0.2878 0.7848 0.1854>}
 dyn_hair_32 {parent=head, loc=<10.5112 9.0139 6.0206>, rot=<0.3376 0.2246 0.9107 0.0792>}
 dyn_hair_33 {parent=head, loc=<11.5673 8.2779 7.2376>, rot=<0.6896 0.1523 0.6927 0.1462>}
 dyn_hair_26 {parent=head, loc=<8.0921 8.403 -4.9273>, rot=<0.056 -0.9166 0.3742 -0.1292>}
 dyn_hair_35 {parent=head, loc=<12.6946 8.04 -4.796>, rot=<-0.15 -0.834 -0.2876 -0.4463>}
 dyn_hair_36 {parent=head, loc=<13.4731 6.758 -5.7876>, rot=<-0.0535 -0.694 -0.0576 -0.7157>}
 dyn_hair_37 {parent=head, loc=<14.6998 5.1405 -4.9415>, rot=<0.1238 -0.5981 0.0891 -0.7868>}
 dyn_hair_38 {parent=head, loc=<14.4662 8.5391 3.5331>, rot=<0.7955 0.1703 0.5239 0.2522>}
 dyn_hair_39 {parent=head, loc=<8.2354 8.195 6.1585>, rot=<0.1761 -0.4326 0.8797 -0.0895>}
 dyn_earring_l {parent=head, loc=<1.1148 0.5334 -6.6437>, rot=<0.1002 0.0111 -0.0011 -0.9949>}
 dyn_hair_01 {parent=head, loc=<7.1752 -6.6048 5.0312>, rot=<0.368 -0.2278 0.8962 -0.0976>}
 dyn_hair_02 {parent=head, loc=<12.5595 -5.4982 4.4661>, rot=<0.5535 -0.348 0.7041 -0.2771>}
 dyn_hair_03 {parent=head, loc=<9.7248 -7.2815 -1.2571>, rot=<0.383 -0.3926 0.8148 -0.1877>}
 dyn_hair_04 {parent=head, loc=<7.6017 -6.7652 -3.1959>, rot=<0.3072 -0.316 0.8905 -0.1126>}
 dyn_hair_05 {parent=head, loc=<15.6275 -1.6879 -2.2429>, rot=<0.2121 -0.6037 0.1659 -0.7504>}
 dyn_hair_06 {parent=head, loc=<10.6581 -5.9022 2.8613>, rot=<0.4734 -0.4689 0.6659 -0.3354>}
 dyn_hair_10 {parent=head, loc=<14.473 1.2559 -5.4805>, rot=<0.1238 -0.7052 0.1213 -0.6875>}
 dyn_hair_11 {parent=head, loc=<9.2305 3.689 -9.3605>, rot=<-0.0381 -0.7511 -0.0498 -0.6572>}
 dyn_hair_12 {parent=head, loc=<7.9537 2.8672 -8.8445>, rot=<0.1515 -0.83 0.2607 -0.4693>}
 dyn_hair_13 {parent=head, loc=<6.1558 4.1592 -8.5133>, rot=<-0.0872 -0.8975 -0.3738 -0.2171>}
 dyn_hair_14 {parent=head, loc=<4.5453 -2.9715 -7.4818>, rot=<0.1326 -0.8145 0.5276 -0.2015>}
 dyn_hair_15 {parent=head, loc=<9.49 0.4405 -8.578>, rot=<0.1863 -0.8125 0.343 -0.433>}
 dyn_hair_16 {parent=head, loc=<12.914 -1.5248 -5.886>, rot=<0.1402 -0.7772 0.1795 -0.5865>}
 dyn_hair_17 {parent=head, loc=<13.2309 -2.8094 -4.1328>, rot=<0.2315 -0.7396 0.3037 -0.5542>}
 dyn_hair_18 {parent=head, loc=<12.7028 -4.0281 -2.7865>, rot=<0.1715 -0.8273 0.3299 -0.4211>}
 dyn_hair_19 {parent=head, loc=<12.231 -4.8781 -0.1669>, rot=<0.5564 -0.4162 0.6087 -0.3829>}
 dyn_hair_20 {parent=head, loc=<13.8754 -2.0617 3.2424>, rot=<0.7023 -0.2407 0.607 -0.2834>}
 dyn_hair_21 {parent=head, loc=<15.6275 -0.0219 1.4364>, rot=<0.7213 -0.2432 0.5674 -0.3141>}
 dyn_hair_29 {parent=head, loc=<9.9948 3.2813 8.3698>, rot=<0.5013 0.0935 0.8588 0.0499>}
 dyn_hair_30 {parent=head, loc=<11.3851 0.3741 7.5999>, rot=<0.0029 -0.7071 -0.0029 -0.7071>}
 dyn_hair_31 {parent=head, loc=<14.1362 1.3807 5.2978>, rot=<0.7768 -0.0105 0.6293 -0.0195>}
 dyn_hair_34 {parent=head, loc=<15.9798 2.6411 4.4975>, rot=<0.8398 -0.0277 0.5399 -0.0506>}
 dyn_earring_r {parent=head, loc=<1.1144 0.5334 6.6437>, rot=<0.0111 -0.1002 0.9949 -0.0011>}
 dyn_hair_08 {parent=head, loc=<6.5663 -4.8442 2.1406>, rot=<0.0041 -0.9979 -0.0003 -0.0645>}
 dyn_hair_07 {parent=head, loc=<6.5663 -4.8442 -4.8014>, rot=<0.1256 -0.9919 0.0022 -0.0182>}
 dyn_hair_09 {parent=head, loc=<6.6984 3.163 -7.0121>, rot=<-0.0159 -0.9995 0.0008 -0.028>}
 dyn_hair_22 {parent=head, loc=<8.1217 2.8686 7.3134>, rot=<0.0688 -0.8182 0.5632 0.0928>}
 dyn_hair_23 {parent=head, loc=<7.4974 -2.5785 7.3134>, rot=<0.0378 -0.8403 0.5365 -0.068>}
]
variables [
 skydiving {type=bool, default=False}
 speed scale {type=double, default=0}
]
